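# Physics asset for the dangling belts on the female medium Sailor Squad Leader body, driving their secondary motion.
# Five capsule chains (front left, two front right, two back right sharing a root) hang from MSTR bones constrained to spine_02 or spine_03.
# Belt links use limited swing with tight or locked twist.
# Pelvis, spine, thigh and left front skirt capsules have fully free constraints, apparently serving as colliders for the belts.
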PhysicsAsset /BRCosmetics/Characters/Player/Female/Medium/Bodies/F_MED_Sailor_Squad_Leader/Meshes/F_MED_Sailor_Squad_Leader_dyn_belts
bodies {
 dyn_bk_beltA_01_r [{shapeType=Capsule, length=2, radius=1, shapeName=dyn_bk_beltA_01_r_capsule, center=<2.6604 0 0>, rotation=<rot 90 0 0>}]
 dyn_bk_beltA_02_r [{shapeType=Capsule, length=2.25, radius=1, shapeName=dyn_bk_beltA_02_r_capsule, center=<2.6451 0 0.2>, rotation=<rot 90 0 0>}]
 dyn_bk_beltA_03_r [{shapeType=Capsule, length=2.75, radius=1, shapeName=dyn_bk_beltA_03_r_capsule, center=<2.9707 0 0.2>, rotation=<rot 90 0 0>}]
 dyn_bk_beltB_01_r [{shapeType=Capsule, length=2.5, radius=1, shapeName=dyn_bk_beltB_01_r_capsule, center=<2.7469 0 0.2596>, rotation=<rot 90 0 0>}]
 dyn_bk_beltB_02_r [{shapeType=Capsule, length=3.25, radius=1.25, shapeName=dyn_bk_beltB_02_r_capsule, center=<2.2992 0 0.6>, rotation=<rot 90 0 0>}]
 dyn_bk_beltB_03_r [{shapeType=Capsule, length=6, radius=1.5, shapeName=dyn_bk_beltB_03_r_capsule, center=<6.8723 0.4 -0.1>, rotation=<rot 85 0 0>}]
 dyn_bk_belt_r_MSTR [{shapeType=Capsule, length=1, radius=1}]
 dyn_ft_beltA_01_r [{shapeType=Capsule, length=0.75, radius=1.5, shapeName=dyn_ft_beltA_01_r_capsule, center=<2 0 0>, rotation=<rot 90 0 0>}]
 dyn_ft_beltA_02_r [{shapeType=Capsule, length=0.75, radius=1.5, shapeName=dyn_ft_beltA_02_r_capsule, center=<2 0 0>, rotation=<rot 90 0 0>}]
 dyn_ft_beltA_03_r [{shapeType=Capsule, length=1.75, radius=1.5, shapeName=dyn_ft_beltA_03_r_capsule, center=<2.6591 0 0>, rotation=<rot 90 0 0>}]
 dyn_ft_beltA_r_MSTR [{shapeType=Capsule, length=1, radius=1}]
 dyn_ft_beltB_01_r [{shapeType=Capsule, length=1.5, radius=0.75, shapeName=dyn_ft_beltB_01_r_capsule, center=<2.4733 0 0>, rotation=<rot 90 0 0>}]
 dyn_ft_beltB_02_r [{shapeType=Capsule, length=1.75, radius=0.75, shapeName=dyn_ft_beltB_02_r_capsule, center=<1.9167 0 0>, rotation=<rot 90 0 0>}]
 dyn_ft_beltB_03_r [{shapeType=Capsule, length=4.75, radius=0.75, shapeName=dyn_ft_beltB_03_r_capsule, center=<3.5808 0 0.1167>, rotation=<rot 90 0 0>}]
 dyn_ft_beltB_r_MSTR [{shapeType=Capsule, length=1, radius=1}]
 dyn_ft_belt_01_l [{shapeType=Capsule, length=1.25, radius=2.5, shapeName=dyn_ft_belt_01_l_capsule, center=<2.6635 0 0.1581>, rotation=<rot 90 0 0>}]
 dyn_ft_belt_02_l [{shapeType=Capsule, length=1.25, radius=2.25, shapeName=dyn_ft_belt_02_l_capsule, center=<3.213 0 0.1727>, rotation=<rot 90 0 0>}]
 dyn_ft_belt_03_l [{shapeType=Capsule, length=2.5, radius=1.75, shapeName=dyn_ft_belt_03_l_capsule, center=<2.5571 0 0>, rotation=<rot 90 0 0>}]
 dyn_ft_belt_04_l [{shapeType=Capsule, length=6.75, radius=1.75, shapeName=dyn_ft_belt_04_l_capsule, center=<4.3181 0 0.6009>, rotation=<rot 90 0 0>}]
 dyn_ft_belt_l_MSTR [{shapeType=Capsule, length=1, radius=1}]
 dyn_ft_skirt_03_l [{shapeType=Capsule, length=8.6891, radius=5.7805, center=<2 -0.0001 -4.6328>, rotation=<rot 0 0 90>}]
 dyn_ft_skirt_04_l [{shapeType=Capsule, length=8.6891, radius=5.7805, center=<2 -0.0001 -4.6328>, rotation=<rot 0 0 90>}]
 pelvis [
  {shapeType=Capsule, length=10.8929, radius=10, center=<-2.1789 -1.8315 -0.0218>, rotation=<rot 0 0 -2.1777>}
  {shapeType=Capsule, length=12.1821, radius=11.156, center=<-8.5375 -1.1767 0.3246>}
 ]
 spine_01 [{shapeType=Capsule, length=7.5642, radius=8.8336, center=<-1.3922 0.7394 0.0181>}]
 spine_02 [{shapeType=Capsule, length=3.2228, radius=9.675, center=<-2.1228 0.3982 -0.0211>, rotation=<rot 0 0 4.9424>}]
 spine_03 [{shapeType=Capsule, length=3.2615, radius=11.206, center=<-1.9925 -0.5607 0>}]
 thigh_l [{shapeType=Capsule, length=27.2733, radius=9.0325, center=<-15.56 -0.0259 -1.4613>, rotation=<rot 87.1755 179.4501 -180>}]
 thigh_r [{shapeType=Capsule, length=25.2733, radius=9, center=<15.6361 1.327 0.1707>, rotation=<rot 87.1753 179.4501 0>}]
}
constraints [
 {
  bone1Name=spine_01
  bone2Name=pelvis
  swing1LimitDegrees=45
  swing1Motion=Free
  swing2LimitDegrees=45
  swing2Motion=Free
  twistLimitDegrees=45
  twistMotion=Free
 }
 {
  bone1Name=spine_02
  bone2Name=spine_01
  swing1LimitDegrees=45
  swing1Motion=Free
  swing2LimitDegrees=45
  swing2Motion=Free
  twistLimitDegrees=45
  twistMotion=Free
 }
 {
  bone1Name=spine_03
  bone2Name=spine_02
  swing1LimitDegrees=45
  swing1Motion=Free
  swing2LimitDegrees=45
  swing2Motion=Free
  twistLimitDegrees=45
  twistMotion=Free
 }
 {
  bone1Name=dyn_bk_belt_r_MSTR
  bone2Name=spine_02
  swing1LimitDegrees=45
  swing1Motion=Free
  swing2LimitDegrees=45
  swing2Motion=Free
  twistLimitDegrees=45
  twistMotion=Free
 }
 {
  bone1Name=dyn_ft_belt_l_MSTR
  bone2Name=spine_02
  swing1LimitDegrees=45
  swing1Motion=Free
  swing2LimitDegrees=45
  swing2Motion=Free
  twistLimitDegrees=45
  twistMotion=Free
 }
 {
  bone1Name=dyn_ft_belt_01_l
  bone2Name=dyn_ft_belt_l_MSTR
  swing1LimitDegrees=25
  swing2LimitDegrees=50
  twistLimitDegrees=2
  twistMotion=Locked
 }
 {
  bone1Name=dyn_ft_belt_02_l
  bone2Name=dyn_ft_belt_01_l
  swing1LimitDegrees=20
  swing2LimitDegrees=30
  twistLimitDegrees=2
  twistMotion=Locked
 }
 {bone1Name=dyn_ft_belt_03_l, bone2Name=dyn_ft_belt_02_l, swing1LimitDegrees=30, swing2LimitDegrees=45, twistLimitDegrees=2}
 {bone1Name=dyn_ft_belt_04_l, bone2Name=dyn_ft_belt_03_l, swing1LimitDegrees=30, swing2LimitDegrees=45, twistLimitDegrees=2}
 {
  bone1Name=dyn_ft_beltA_r_MSTR
  bone2Name=spine_03
  swing1LimitDegrees=45
  swing1Motion=Free
  swing2LimitDegrees=45
  swing2Motion=Free
  twistLimitDegrees=45
  twistMotion=Free
 }
 {bone1Name=dyn_ft_beltA_01_r, bone2Name=dyn_ft_beltA_r_MSTR, swing1LimitDegrees=30, swing2LimitDegrees=35, twistLimitDegrees=2}
 {bone1Name=dyn_ft_beltA_02_r, bone2Name=dyn_ft_beltA_01_r, swing1LimitDegrees=30, swing2LimitDegrees=45, twistLimitDegrees=2}
 {bone1Name=dyn_ft_beltA_03_r, bone2Name=dyn_ft_beltA_02_r, swing1LimitDegrees=30, swing2LimitDegrees=45, twistLimitDegrees=2}
 {
  bone1Name=dyn_ft_beltB_r_MSTR
  bone2Name=spine_03
  swing1LimitDegrees=45
  swing1Motion=Free
  swing2LimitDegrees=45
  swing2Motion=Free
  twistLimitDegrees=45
  twistMotion=Free
 }
 {bone1Name=dyn_ft_beltB_01_r, bone2Name=dyn_ft_beltB_r_MSTR, swing1LimitDegrees=30, swing2LimitDegrees=45, twistLimitDegrees=2}
 {bone1Name=dyn_ft_beltB_02_r, bone2Name=dyn_ft_beltB_01_r, swing1LimitDegrees=30, swing2LimitDegrees=45, twistLimitDegrees=2}
 {bone1Name=dyn_ft_beltB_03_r, bone2Name=dyn_ft_beltB_02_r, swing1LimitDegrees=30, swing2LimitDegrees=45, twistLimitDegrees=2}
 {
  bone1Name=thigh_l
  bone2Name=pelvis
  swing1LimitDegrees=45
  swing1Motion=Free
  swing2LimitDegrees=45
  swing2Motion=Free
  twistLimitDegrees=45
  twistMotion=Free
 }
 {
  bone1Name=thigh_r
  bone2Name=pelvis
  swing1LimitDegrees=45
  swing1Motion=Free
  swing2LimitDegrees=45
  swing2Motion=Free
  twistLimitDegrees=45
  twistMotion=Free
 }
 {
  bone1Name=dyn_bk_beltA_01_r
  bone2Name=dyn_bk_belt_r_MSTR
  swing1LimitDegrees=10
  swing2LimitDegrees=30
  twistLimitDegrees=2
  twistMotion=Locked
 }
 {bone1Name=dyn_bk_beltA_02_r, bone2Name=dyn_bk_beltA_01_r, swing1LimitDegrees=10, swing2LimitDegrees=30, twistLimitDegrees=2}
 {bone1Name=dyn_bk_beltA_03_r, bone2Name=dyn_bk_beltA_02_r, swing1LimitDegrees=15, swing2LimitDegrees=45, twistLimitDegrees=2}
 {
  bone1Name=dyn_bk_beltB_01_r
  bone2Name=dyn_bk_belt_r_MSTR
  swing1LimitDegrees=10
  swing2LimitDegrees=20
  twistLimitDegrees=2
  twistMotion=Locked
 }
 {bone1Name=dyn_bk_beltB_02_r, bone2Name=dyn_bk_beltB_01_r, swing1LimitDegrees=10, swing2LimitDegrees=30, twistLimitDegrees=2}
 {bone1Name=dyn_bk_beltB_03_r, bone2Name=dyn_bk_beltB_02_r, swing1LimitDegrees=15, swing2LimitDegrees=45, twistLimitDegrees=2}
 {
  bone1Name=dyn_ft_skirt_03_l
  bone2Name=pelvis
  swing1LimitDegrees=45
  swing1Motion=Free
  swing2LimitDegrees=45
  swing2Motion=Free
  twistLimitDegrees=45
  twistMotion=Free
 }
 {
  bone1Name=dyn_ft_skirt_04_l
  bone2Name=dyn_ft_skirt_03_l
  swing1LimitDegrees=45
  swing1Motion=Free
  swing2LimitDegrees=45
  swing2Motion=Free
  twistLimitDegrees=45
  twistMotion=Free
 }
]
