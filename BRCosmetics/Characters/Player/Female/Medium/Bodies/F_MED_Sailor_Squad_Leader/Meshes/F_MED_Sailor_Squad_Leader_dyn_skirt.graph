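# Physics asset for the skirt dynamics of the female medium Sailor Squad Leader body.
# Capsules on pelvis, lower spine and both thighs form body volumes around three skirt bone chains.
# Front-left and back-left chains hang from pelvis through small master bodies, while the shorter back-right chain hangs from the right thigh.
# Master and body constraints swing and twist freely, while the chain links carry tighter angular limits.
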PhysicsAsset /BRCosmetics/Characters/Player/Female/Medium/Bodies/F_MED_Sailor_Squad_Leader/Meshes/F_MED_Sailor_Squad_Leader_dyn_skirt
bodies {
 dyn_bk_skirt_01_l [{shapeType=Capsule, length=8.5, radius=1.5, shapeName=dyn_bk_skirt_01_l_capsule, center=<2.1635 0 0>, rotation=<rot 0 0 -90>}]
 dyn_bk_skirt_01_r [{shapeType=Capsule, length=10, radius=1.5, center=<3.3585 -3.3313 -2.0627>, rotation=<rot 0 -5 -120>}]
 dyn_bk_skirt_02_l [{shapeType=Capsule, length=8.25, radius=1.25, shapeName=dyn_bk_skirt_02_l_capsule, rotation=<rot 0 0 -90>}]
 dyn_bk_skirt_02_r [{shapeType=Capsule, length=10, radius=1.5, center=<2.7769 -3.5641 -1.476>, rotation=<rot 0 -10 -110>}]
 dyn_bk_skirt_03_l [{shapeType=Capsule, length=9.5, radius=1.5, shapeName=dyn_bk_skirt_03_l_capsule, center=<3.6971 0 0>, rotation=<rot 0 0 -90>}]
 dyn_bk_skirt_MSTR [{shapeType=Capsule, length=0.505, radius=0.505, rotation=<rot 0 -90 0>}]
 dyn_bk_skirt_r_MSTR [{shapeType=Capsule, length=0.505, radius=0.505, rotation=<rot 0 -90 0>}]
 dyn_ft_skirt_01_l [{shapeType=Capsule, length=8.25, radius=2.75, center=<4.3125 1.6548 0.6106>, rotation=<rot 0 0 -90>}]
 dyn_ft_skirt_02_l [{shapeType=Capsule, length=12.5, radius=1.75, center=<5.132 2.6013 0>, rotation=<rot 0 0 -90>}]
 dyn_ft_skirt_03_l [{shapeType=Capsule, length=12.5, radius=1.5, center=<4 1.9653 0>, rotation=<rot 0 0 -90>}]
 dyn_ft_skirt_04_l [{shapeType=Capsule, length=9.75, radius=1.5, center=<3 0.4936 0>, rotation=<rot 0 0 -90>}]
 dyn_ft_skirt_l_MSTR [{shapeType=Capsule, length=0.505, radius=0.505, rotation=<rot 0 -90 0>}]
 pelvis [{shapeType=Capsule, length=10.5674, radius=9.9603, center=<-6.6408 -1.7424 0>}]
 spine_01 [{shapeType=Capsule, length=7.5674, radius=9.9603, center=<-3.2383 -2.073 0.0001>}]
 thigh_l [{shapeType=Capsule, length=21.2733, radius=8.0325, center=<-15.5283 -0.614 -1.0076>, rotation=<rot 84.8117 -178.0239 180>}]
 thigh_r [{shapeType=Capsule, length=21.5233, radius=7.7825, center=<15.5429 0.1898 1.0076>, rotation=<rot 84.8125 -178.0239 0>}]
}
constraints [
 {
  bone1Name=thigh_l
  bone2Name=pelvis
  swing1LimitDegrees=45
  swing1Motion=Free
  swing2LimitDegrees=45
  swing2Motion=Free
  twistLimitDegrees=45
  twistMotion=Free
 }
 {
  bone1Name=dyn_ft_skirt_l_MSTR
  bone2Name=pelvis
  swing1LimitDegrees=45
  swing1Motion=Free
  swing2LimitDegrees=45
  swing2Motion=Free
  twistLimitDegrees=45
  twistMotion=Free
 }
 {bone1Name=dyn_ft_skirt_01_l, bone2Name=dyn_ft_skirt_l_MSTR, swing1LimitDegrees=10, swing2LimitDegrees=20, twistLimitDegrees=30}
 {bone1Name=dyn_ft_skirt_02_l, bone2Name=dyn_ft_skirt_01_l, swing1LimitDegrees=10, swing2LimitDegrees=50, twistLimitDegrees=20}
 {bone1Name=dyn_ft_skirt_03_l, bone2Name=dyn_ft_skirt_02_l, swing1LimitDegrees=10, swing2LimitDegrees=20, twistLimitDegrees=20}
 {bone1Name=dyn_ft_skirt_04_l, bone2Name=dyn_ft_skirt_03_l, swing1LimitDegrees=10, swing2LimitDegrees=20, twistLimitDegrees=20}
 {
  bone1Name=spine_01
  bone2Name=pelvis
  swing1LimitDegrees=45
  swing1Motion=Free
  swing2LimitDegrees=45
  swing2Motion=Free
  twistLimitDegrees=45
  twistMotion=Free
 }
 {
  bone1Name=thigh_r
  bone2Name=pelvis
  swing1LimitDegrees=45
  swing1Motion=Free
  swing2LimitDegrees=45
  swing2Motion=Free
  twistLimitDegrees=45
  twistMotion=Free
 }
 {
  bone1Name=dyn_bk_skirt_MSTR
  bone2Name=pelvis
  swing1LimitDegrees=45
  swing1Motion=Free
  swing2LimitDegrees=45
  swing2Motion=Free
  twistLimitDegrees=45
  twistMotion=Free
 }
 {bone1Name=dyn_bk_skirt_01_l, bone2Name=dyn_bk_skirt_MSTR, swing1LimitDegrees=10, swing2LimitDegrees=20, twistLimitDegrees=10}
 {bone1Name=dyn_bk_skirt_02_l, bone2Name=dyn_bk_skirt_01_l, swing1LimitDegrees=10, swing2LimitDegrees=20, twistLimitDegrees=5}
 {bone1Name=dyn_bk_skirt_03_l, bone2Name=dyn_bk_skirt_02_l, swing1LimitDegrees=10, swing2LimitDegrees=20, twistLimitDegrees=5}
 {
  bone1Name=dyn_bk_skirt_r_MSTR
  bone2Name=thigh_r
  swing1LimitDegrees=45
  swing1Motion=Free
  swing2LimitDegrees=45
  swing2Motion=Free
  twistLimitDegrees=45
  twistMotion=Free
 }
 {bone1Name=dyn_bk_skirt_01_r, bone2Name=dyn_bk_skirt_r_MSTR, swing1LimitDegrees=5, swing2LimitDegrees=20, twistLimitDegrees=5}
 {bone1Name=dyn_bk_skirt_02_r, bone2Name=dyn_bk_skirt_01_r, swing1LimitDegrees=5, swing2LimitDegrees=20, twistLimitDegrees=5}
]
