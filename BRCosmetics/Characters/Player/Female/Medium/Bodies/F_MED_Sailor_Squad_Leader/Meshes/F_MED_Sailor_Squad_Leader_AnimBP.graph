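# Character-part AnimBP for the female Sailor Squad Leader body, driving skirt, belt, bag and accessory secondary motion.
# It copies the attached parent's pose, then applies weighted constraints and bag look-ats into a shared cache.
# bRigidBodyEnabled selects either RigidBody and AnimDynamics simulation or a look-at and trail-controller branch.
# Each update derives movement, thigh-lift and skydiving alphas plus a gravity override, via a deprecated function, to tune both branches.
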AnimBlueprint /BRCosmetics/Characters/Player/Female/Medium/Bodies/F_MED_Sailor_Squad_Leader/Meshes/F_MED_Sailor_Squad_Leader_AnimBP
target_skeleton=&/F_MED_Sailor_Squad_Leader_Skeleton
graphs [
 AnimGraph {nodes [
   Output Pose(Result=Blend Poses by bool(BlendPose_0=Use cached pose 'rigidCache'(), BlendPose_1=Use cached pose 'trailCache'(), bActiveValue=Get bRigidBodyEnabled()) {blendType=Linear}) {layerGroup=None}
   n1 = Anim Dynamics - Bone: dyn_ring(ComponentPose=Local To Component(LocalPose=Use cached pose 'startCache'()), Alpha=1, ExternalForce="0,0,0") {
    angularDampingOverride=0.999
    bGravityOverrideInSimSpace=true
    bOverrideAngularDamping=true
    bUseGravityOverride=true
    alphaBoolBlend {blendOption=Linear}
    boundBone {boneName=dyn_ring}
    componentAppliedLinearAccClamp=<3000 3000 3000>
    componentLinearAccScale=<0.3 0.3 0.3>
    componentLinearVelScale=<0.3 0.3 0.3>
    gravityOverride {z=-980}
    physicsBodyDefinitions [{
      sphereCollisionRadius=0
      boundBone {boneName=dyn_ring}
      boxExtents=<10 12 12>
      constraintSetup {angularLimitsMax=<10 5 30>, angularLimitsMin=<-10 -75 -30>, angularTarget=<0 0 0>}
      localJointOffset=<2 0.2 0.2>
     }]
   }
   n2 = "Transform (Modify) Bone - Bone: dyn_ft_skirt_l_MSTR"(ComponentPose=@n1, Alpha=0, Translation="3.388,9.049,2.350", Rotation="60.008,179.701,179.520", Scale="1,1,1") {
    rotationMode=BMM_Additive
    rotationSpace=BCS_BoneSpace
    translationMode=BMM_Additive
    translationSpace=BCS_BoneSpace
    alphaBoolBlend {blendOption=Linear}
    boneToModify {boneName=dyn_ft_skirt_l_MSTR}
    rotation=<rot 60.008 179.701 179.52>
    translation=<3.388 9.049 2.35>
   }
   n3 = "Transform (Modify) Bone - Bone: dyn_ft_belt_l_MSTR"(
    ComponentPose=@n2
    Alpha=Get thigh_rot_up_l() -> (thigh_rot_up_l:double)
    Translation="-8.280,11.591,7.740"
    Rotation="36.471,179.783,-178.881"
    Scale="1,1,1"
   ) {
    rotationMode=BMM_Additive
    rotationSpace=BCS_BoneSpace
    translationMode=BMM_Additive
    translationSpace=BCS_BoneSpace
    alphaBoolBlend {blendOption=Linear}
    boneToModify {boneName=dyn_ft_belt_l_MSTR}
    rotation=<rot 36.471 179.783 -178.881>
    translation=<-8.28 11.591 7.74>
   }
   n4 = Get GravityOverride()
   n5 = RigidBody!(ComponentPose=@n3, Alpha=Get Skydiving 05() -> (Skydiving 05:double), OverrideWorldGravity:Vector=@n4, ExternalForce) -> (Pose) {
    bClampLinearTranslationLimitToRefPose=true
    bOverrideWorldGravity=true
    overridePhysicsAsset=&/F_MED_Sailor_Squad_Leader_dyn_skirt
    simulationSpace=BaseBoneSpace
    alphaBoolBlend {blendOption=Linear}
    alphaScaleBiasClamp {bInterpResult=true}
    baseBoneRef {boneName=pelvis}
    componentAppliedLinearAccClamp=<3000 3000 3000>
    overrideWorldGravity {z=-980}
    simSpaceSettings {maxAngularAcceleration=1000, maxAngularVelocity=5, maxLinearAcceleration=1000, maxLinearVelocity=100, worldAlpha=1}
   }
   n6 = "Transform (Modify) Bone - Bone: dyn_ft_skirt_01_l"!(ComponentPose=@n5, bAlphaBoolEnabled:bool=Get bIsCrouching(), Translation="0,0,0", Rotation="0,0,-20", Scale="1,1,1") -> (Pose) {
    alphaInputType=Bool
    rotationMode=BMM_Additive
    alphaBoolBlend {blendOption=Linear}
    boneToModify {boneName=dyn_ft_skirt_01_l}
    rotation {roll=-20}
   }
   n7 = RigidBody(ComponentPose=@n6, Alpha=Get Skydiving 025() -> (Skydiving 025:double), OverrideWorldGravity=@n4) {
    bClampLinearTranslationLimitToRefPose=true
    bOverrideWorldGravity=true
    overridePhysicsAsset=&/F_MED_Sailor_Squad_Leader_dyn_belts
    simulationSpace=BaseBoneSpace
    alphaBoolBlend {blendOption=Linear}
    alphaScaleBiasClamp {bInterpResult=true}
    baseBoneRef {boneName=pelvis}
    componentAppliedLinearAccClamp=<3000 3000 3000>
    componentLinearAccScale=<0.25 0.25 0.3>
    componentLinearVelScale=<0.3 0.3 0.3>
    overrideWorldGravity {z=-980}
    simSpaceSettings {maxAngularAcceleration=800, maxAngularVelocity=3, maxLinearAcceleration=500, maxLinearVelocity=10}
   }
   n8 = "Transform (Modify) Bone - Bone: dyn_bk_belt_r_MSTR"!(ComponentPose=@n7, bAlphaBoolEnabled:bool=Get bIsSkydiveFloating(), Translation="0,0,0", Rotation="40.000,-0.000,0.000", Scale="1,1,1") -> (Pose) {
    alphaInputType=Bool
    rotationMode=BMM_Additive
    rotationSpace=BCS_BoneSpace
    alphaBoolBlend {blendOption=Linear}
    alphaScaleBiasClamp {bInterpResult=true}
    boneToModify {boneName=dyn_bk_belt_r_MSTR}
    rotation=<rot 40 0 0>
   }
   n9 = "Transform (Modify) Bone - Bone: dyn_bk_beltB_01_r"(ComponentPose=@n8, Alpha=1, Translation="0,0,0", Rotation="2.511,5.001,0.437", Scale="1,1,1") {
    rotationMode=BMM_Additive
    rotationSpace=BCS_BoneSpace
    alphaBoolBlend {blendOption=Linear}
    alphaScaleBiasClamp {bInterpResult=true}
    boneToModify {boneName=dyn_bk_beltB_01_r}
    rotation=<rot 2.511 5.001 0.437>
   }
   n10 = "Transform (Modify) Bone - Bone: dyn_bk_beltA_01_r"(ComponentPose=@n9, Alpha=1, Translation="0,0,0", Rotation="4.531,-4.945,0.442", Scale="1,1,1") {
    rotationMode=BMM_Additive
    rotationSpace=BCS_BoneSpace
    alphaBoolBlend {blendOption=Linear}
    alphaScaleBiasClamp {bInterpResult=true}
    boneToModify {boneName=dyn_bk_beltA_01_r}
    rotation=<rot 4.531 -4.945 0.442>
   }
   n11 = "Transform (Modify) Bone - Bone: dyn_bk_beltB_02_r"(ComponentPose=@n10, Alpha=1, Translation="0,0,0", Rotation="-7.249,-0.000,0.000", Scale="1,1,1") {
    rotationMode=BMM_Additive
    rotationSpace=BCS_BoneSpace
    alphaBoolBlend {blendOption=Linear}
    boneToModify {boneName=dyn_bk_beltB_02_r}
    rotation=<rot -7.249 0 0>
   }
   n12 = "Transform (Modify) Bone - Bone: dyn_bk_beltA_02_r"(ComponentPose=@n11, Alpha=1, Translation="-0.116,0.617,-0.781", Rotation="-8.446,-2.936,0.431", Scale="1,1,1") {
    rotationMode=BMM_Additive
    rotationSpace=BCS_BoneSpace
    translationMode=BMM_Additive
    translationSpace=BCS_BoneSpace
    alphaBoolBlend {blendOption=Linear}
    boneToModify {boneName=dyn_bk_beltA_02_r}
    rotation=<rot -8.446 -2.936 0.431>
    translation=<-0.116 0.617 -0.781>
   }
   n13 = "Transform (Modify) Bone - Bone: dyn_bk_beltB_03_r"(ComponentPose=@n12, Alpha=1, Translation="0,0,0", Rotation="3.708,-0.000,-0.000", Scale="1,1,1") {
    rotationMode=BMM_Additive
    rotationSpace=BCS_BoneSpace
    alphaBoolBlend {blendOption=Linear}
    boneToModify {boneName=dyn_bk_beltB_03_r}
    rotation=<rot 3.708 0 0>
   }
   n14 = "Transform (Modify) Bone - Bone: dyn_bk_beltA_03_r"(ComponentPose=@n13, Alpha=1, Translation="-0.062,0.001,0.393", Rotation="5.106,6.338,1.607", Scale="1,1,1") {
    rotationMode=BMM_Additive
    rotationSpace=BCS_BoneSpace
    translationMode=BMM_Additive
    translationSpace=BCS_BoneSpace
    alphaBoolBlend {blendOption=Linear}
    boneToModify {boneName=dyn_bk_beltA_03_r}
    rotation=<rot 5.106 6.338 1.607>
    translation=<-0.0619 0.001 0.3927>
   }
   n15 = Anim Dynamics - Bone: dyn_zipperA_01_l(ComponentPose=@n14, Alpha=1, ExternalForce="0,0,0") {
    angularDampingOverride=0.999
    bGravityOverrideInSimSpace=true
    bOverrideAngularDamping=true
    bUseGravityOverride=true
    alphaBoolBlend {blendOption=Linear}
    boundBone {boneName=dyn_zipperA_01_l}
    componentAppliedLinearAccClamp=<3000 3000 3000>
    componentLinearAccScale=<0.3 0.3 0.3>
    componentLinearVelScale=<0.3 0.3 0.3>
    gravityOverride {z=-980}
    physicsBodyDefinitions [{
      sphereCollisionRadius=0
      boundBone {boneName=dyn_zipperA_01_l}
      boxExtents=<10 12 12>
      constraintSetup {angularLimitsMax=<10 5 30>, angularLimitsMin=<-10 -75 -30>, angularTarget=<0 0 0>}
      localJointOffset=<2 0.2 0.2>
     }]
   }
   n16 = Anim Dynamics - Bone: dyn_zipperB_01_l(ComponentPose=@n15, Alpha=1, ExternalForce="0,0,0") {
    angularDampingOverride=0.999
    bGravityOverrideInSimSpace=true
    bOverrideAngularDamping=true
    bUseGravityOverride=true
    alphaBoolBlend {blendOption=Linear}
    boundBone {boneName=dyn_zipperB_01_l}
    componentAppliedLinearAccClamp=<3000 3000 3000>
    componentLinearAccScale=<0.3 0.3 0.3>
    componentLinearVelScale=<0.3 0.3 0.3>
    gravityOverride {z=-980}
    physicsBodyDefinitions [{
      sphereCollisionRadius=0
      boundBone {boneName=dyn_zipperB_01_l}
      boxExtents=<10 12 12>
      constraintSetup {angularLimitsMax=<10 5 30>, angularLimitsMin=<-10 -75 -30>, angularTarget=<0 0 0>}
      localJointOffset=<2 0.2 0.2>
     }]
   }
   n17 = Anim Dynamics - Bone: dyn_zipperA_01_r(ComponentPose=@n16, Alpha=1, ExternalForce="0,0,0") {
    angularDampingOverride=0.999
    bGravityOverrideInSimSpace=true
    bOverrideAngularDamping=true
    bUseGravityOverride=true
    alphaBoolBlend {blendOption=Linear}
    boundBone {boneName=dyn_zipperA_01_r}
    componentAppliedLinearAccClamp=<3000 3000 3000>
    componentLinearAccScale=<0.3 0.3 0.3>
    componentLinearVelScale=<0.3 0.3 0.3>
    gravityOverride {z=-980}
    physicsBodyDefinitions [{
      sphereCollisionRadius=0
      boundBone {boneName=dyn_zipperA_01_r}
      boxExtents=<10 12 12>
      constraintSetup {angularLimitsMax=<10 5 30>, angularLimitsMin=<-10 -75 -30>, angularTarget=<0 0 0>}
      localJointOffset=<2 0.2 0.2>
     }]
   }
   n18 = Anim Dynamics - Bone: dyn_zipperB_01_r(ComponentPose=@n17, Alpha=1, ExternalForce="0,0,0") {
    angularDampingOverride=0.999
    bGravityOverrideInSimSpace=true
    bOverrideAngularDamping=true
    bUseGravityOverride=true
    alphaBoolBlend {blendOption=Linear}
    boundBone {boneName=dyn_zipperB_01_r}
    componentAppliedLinearAccClamp=<3000 3000 3000>
    componentLinearAccScale=<0.3 0.3 0.3>
    componentLinearVelScale=<0.3 0.3 0.3>
    gravityOverride {z=-980}
    physicsBodyDefinitions [{
      sphereCollisionRadius=0
      boundBone {boneName=dyn_zipperB_01_r}
      boxExtents=<10 12 12>
      constraintSetup {angularLimitsMax=<10 5 30>, angularLimitsMin=<-10 -75 -30>, angularTarget=<0 0 0>}
      localJointOffset=<2 0.2 0.2>
     }]
   }
   Save cached pose 'rigidCache'(Pose=Component To Local(ComponentPose=@n18)) {cacheName=rigidCache}
   n19 = "Constraint - dyn_bag_01_r_MSTR (PPPPPP)"!(
    ComponentPose=Local To Component(LocalPose=Copy Pose From Mesh() {bUseAttachedParent=true})
    Alpha=1
    ConstraintWeights_0=0.0117
    ConstraintWeights_1:float=0.0116
    ConstraintWeights_2:float=0.0208
    ConstraintWeights_3:float=0.1601
    ConstraintWeights_4:float=0.4986
    ConstraintWeights_5:float=0.2973
   ) -> (Pose) {
    alphaBoolBlend {blendOption=Linear}
    boneToModify {boneName=dyn_bag_01_r_MSTR}
    constraintSetup [
     {targetBone {boneName=upperarm_twist_01_r}}
     {targetBone {boneName=spine_01}}
     {targetBone {boneName=spine_02}}
     {targetBone {boneName=spine_03}}
     {targetBone {boneName=spine_04}}
     {targetBone {boneName=spine_05}}
    ]
    constraintWeights [1,1,1,1,1,1]
   }
   n20 = "Constraint - dyn_bag_01_r_TGT (PPPPPP)"!(
    ComponentPose=@n19
    Alpha=1
    ConstraintWeights_0=0.0957
    ConstraintWeights_1:float=0.4338
    ConstraintWeights_2:float=0.2033
    ConstraintWeights_3:float=0.1495
    ConstraintWeights_4:float=0.0673
    ConstraintWeights_5:float=0.0505
   ) -> (Pose) {
    alphaBoolBlend {blendOption=Linear}
    boneToModify {boneName=dyn_bag_01_r_TGT}
    constraintSetup [
     {targetBone {boneName=spine_01}}
     {targetBone {boneName=spine_02}}
     {targetBone {boneName=spine_03}}
     {targetBone {boneName=spine_04}}
     {targetBone {boneName=spine_05}}
     {targetBone {boneName=pelvis}}
    ]
    constraintWeights [1,1,1,1,1,1]
   }
   n21 = Look At - Bone: dyn_bag_01_r(ComponentPose=@n20, Alpha=1) {
    bUseLookUpAxis=true
    alphaBoolBlend {blendOption=Linear}
    boneToModify {boneName=dyn_bag_01_r}
    lookAtLocation {x=0}
    lookAtTarget {boneReference {boneName=dyn_bag_01_r_TGT}}
    lookAt_Axis {axis=<1 0>}
    lookUp_Axis {axis=<0 1>}
   }
   n22 = "Constraint - dyn_bag_02_r_MSTR (PPPPPPP)"!(
    ComponentPose=@n21
    Alpha=1
    ConstraintWeights_0=0.0561
    ConstraintWeights_1:float=0.1272
    ConstraintWeights_2:float=0.2576
    ConstraintWeights_3:float=0.4002
    ConstraintWeights_4:float=0.1184
    ConstraintWeights_5:float=0.0282
    ConstraintWeights_6:float=0.0123
   ) -> (Pose) {
    alphaBoolBlend {blendOption=Linear}
    boneToModify {boneName=dyn_bag_02_r_MSTR}
    constraintSetup [
     {targetBone {boneName=spine_01}}
     {targetBone {boneName=spine_02}}
     {targetBone {boneName=spine_03}}
     {targetBone {boneName=spine_04}}
     {targetBone {boneName=spine_05}}
     {targetBone {boneName=clavicle_r}}
     {targetBone {boneName=pelvis}}
    ]
    constraintWeights [1,1,1,1,1,1,1]
   }
   n23 = "Constraint - dyn_bag_02_r_TGT (PPPPPP)"!(
    ComponentPose=@n22
    Alpha=1
    ConstraintWeights_0=0.157
    ConstraintWeights_1:float=0.4233
    ConstraintWeights_2:float=0.1204
    ConstraintWeights_3:float=0.1002
    ConstraintWeights_4:float=0.0231
    ConstraintWeights_5:float=0.1761
   ) -> (Pose) {
    alphaBoolBlend {blendOption=Linear}
    boneToModify {boneName=dyn_bag_02_r_TGT}
    constraintSetup [
     {targetBone {boneName=spine_01}}
     {targetBone {boneName=spine_02}}
     {targetBone {boneName=spine_03}}
     {targetBone {boneName=spine_04}}
     {targetBone {boneName=spine_05}}
     {targetBone {boneName=pelvis}}
    ]
    constraintWeights [1,1,1,1,1,1]
   }
   n24 = Look At - Bone: dyn_bag_02_r(ComponentPose=@n23, Alpha=1) {
    bUseLookUpAxis=true
    alphaBoolBlend {blendOption=Linear}
    boneToModify {boneName=dyn_bag_02_r}
    lookAtLocation {x=0}
    lookAtTarget {boneReference {boneName=dyn_bag_02_r_TGT}}
    lookAt_Axis {axis=<1 0>}
    lookUp_Axis {axis=<0 1>}
   }
   n25 = "Constraint - dyn_bag_01_l_MSTR (PPPPPPP)"!(
    ComponentPose=@n24
    Alpha=1
    ConstraintWeights_0=0.0204
    ConstraintWeights_1:float=0.0152
    ConstraintWeights_2:float=0.0264
    ConstraintWeights_3:float=0.1638
    ConstraintWeights_4:float=0.4921
    ConstraintWeights_5:float=0.243
    ConstraintWeights_6:float=0.0391
   ) -> (Pose) {
    alphaBoolBlend {blendOption=Linear}
    boneToModify {boneName=dyn_bag_01_l_MSTR}
    constraintSetup [
     {targetBone {boneName=clavicle_l}}
     {targetBone {boneName=spine_01}}
     {targetBone {boneName=spine_02}}
     {targetBone {boneName=spine_03}}
     {targetBone {boneName=spine_04}}
     {targetBone {boneName=spine_05}}
     {targetBone {boneName=upperarm_twist_01_l}}
    ]
    constraintWeights [1,1,1,1,1,1,1]
   }
   n26 = "Constraint - dyn_bag_01_l_TGT (PPPPPP)"!(
    ComponentPose=@n25
    Alpha=1
    ConstraintWeights_0=0.1055
    ConstraintWeights_1:float=0.4463
    ConstraintWeights_2:float=0.1919
    ConstraintWeights_3:float=0.1434
    ConstraintWeights_4:float=0.0557
    ConstraintWeights_5:float=0.0573
   ) -> (Pose) {
    alphaBoolBlend {blendOption=Linear}
    boneToModify {boneName=dyn_bag_01_l_TGT}
    constraintSetup [
     {targetBone {boneName=spine_01}}
     {targetBone {boneName=spine_02}}
     {targetBone {boneName=spine_03}}
     {targetBone {boneName=spine_04}}
     {targetBone {boneName=spine_05}}
     {targetBone {boneName=pelvis}}
    ]
    constraintWeights [1,1,1,1,1,1]
   }
   n27 = Look At - Bone: dyn_bag_01_l(ComponentPose=@n26, Alpha=1) {
    bUseLookUpAxis=true
    alphaBoolBlend {blendOption=Linear}
    boneToModify {boneName=dyn_bag_01_l}
    lookAtLocation {x=0}
    lookAtTarget {boneReference {boneName=dyn_bag_01_l_TGT}}
    lookAt_Axis {axis=<1 0>}
    lookUp_Axis {axis=<0 1>}
   }
   n28 = "Constraint - dyn_bag_02_l_MSTR (PPPPPPP)"!(
    ComponentPose=@n27
    Alpha=1
    ConstraintWeights_0=0.0357
    ConstraintWeights_1:float=0.0229
    ConstraintWeights_2:float=0.0614
    ConstraintWeights_3:float=0.2013
    ConstraintWeights_4:float=0.4508
    ConstraintWeights_5:float=0.1655
    ConstraintWeights_6:float=0.0625
   ) -> (Pose) {
    alphaBoolBlend {blendOption=Linear}
    boneToModify {boneName=dyn_bag_02_l_MSTR}
    constraintSetup [
     {targetBone {boneName=clavicle_l}}
     {targetBone {boneName=spine_01}}
     {targetBone {boneName=spine_02}}
     {targetBone {boneName=spine_03}}
     {targetBone {boneName=spine_04}}
     {targetBone {boneName=spine_05}}
     {targetBone {boneName=upperarm_twist_01_l}}
    ]
    constraintWeights [1,1,1,1,1,1,1]
   }
   n29 = "Constraint - dyn_bag_02_l_TGT (PPPPPP)"!(
    ComponentPose=@n28
    Alpha=1
    ConstraintWeights_0=0.1227
    ConstraintWeights_1:float=0.4186
    ConstraintWeights_2:float=0.1818
    ConstraintWeights_3:float=0.1466
    ConstraintWeights_4:float=0.0413
    ConstraintWeights_5:float=0.0889
   ) -> (Pose) {
    alphaBoolBlend {blendOption=Linear}
    boneToModify {boneName=dyn_bag_02_l_TGT}
    constraintSetup [
     {targetBone {boneName=spine_01}}
     {targetBone {boneName=spine_02}}
     {targetBone {boneName=spine_03}}
     {targetBone {boneName=spine_04}}
     {targetBone {boneName=spine_05}}
     {targetBone {boneName=pelvis}}
    ]
    constraintWeights [1,1,1,1,1,1]
   }
   n30 = Look At - Bone: dyn_bag_02_l(ComponentPose=@n29, Alpha=1) {
    bUseLookUpAxis=true
    alphaBoolBlend {blendOption=Linear}
    boneToModify {boneName=dyn_bag_02_l}
    lookAtLocation {x=0}
    lookAtTarget {boneReference {boneName=dyn_bag_02_l_TGT}}
    lookAt_Axis {axis=<1 0>}
    lookUp_Axis {axis=<0 1>}
   }
   n31 = "Constraint - dyn_ft_beltA_r_MSTR (PPPPPP)"!(
    ComponentPose=@n30
    Alpha=1
    ConstraintWeights_0=0.0168
    ConstraintWeights_1:float=0.0802
    ConstraintWeights_2:float=0.2124
    ConstraintWeights_3:float=0.5254
    ConstraintWeights_4:float=0.1582
    ConstraintWeights_5:float=0.007
   ) -> (Pose) {
    alphaBoolBlend {blendOption=Linear}
    boneToModify {boneName=dyn_ft_beltA_r_MSTR}
    constraintSetup [
     {targetBone {boneName=spine_01}}
     {targetBone {boneName=spine_02}}
     {targetBone {boneName=spine_03}}
     {targetBone {boneName=spine_04}}
     {targetBone {boneName=spine_05}}
     {targetBone {boneName=pelvis}}
    ]
    constraintWeights [1,1,1,1,1,1]
   }
   n32 = "Constraint - dyn_ft_beltB_r_MSTR (PPPPPP)"!(
    ComponentPose=@n31
    Alpha=1
    ConstraintWeights_0=0.1164
    ConstraintWeights_1:float=0.4659
    ConstraintWeights_2:float=0.0779
    ConstraintWeights_3:float=0.1876
    ConstraintWeights_4:float=0.0548
    ConstraintWeights_5:float=0.0974
   ) -> (Pose) {
    alphaBoolBlend {blendOption=Linear}
    boneToModify {boneName=dyn_ft_beltB_r_MSTR}
    constraintSetup [
     {targetBone {boneName=spine_01}}
     {targetBone {boneName=spine_02}}
     {targetBone {boneName=spine_03}}
     {targetBone {boneName=spine_04}}
     {targetBone {boneName=spine_05}}
     {targetBone {boneName=pelvis}}
    ]
    constraintWeights [1,1,1,1,1,1]
   }
   n33 = "Constraint - dyn_ft_belt_l_MSTR (PPPPPP)"!(
    ComponentPose=@n32
    Alpha=1
    ConstraintWeights_0=0.133
    ConstraintWeights_1:float=0.4191
    ConstraintWeights_2:float=0.1616
    ConstraintWeights_3:float=0.1163
    ConstraintWeights_4:float=0.0334
    ConstraintWeights_5:float=0.1366
   ) -> (Pose) {
    alphaBoolBlend {blendOption=Linear}
    boneToModify {boneName=dyn_ft_belt_l_MSTR}
    constraintSetup [
     {targetBone {boneName=spine_01}}
     {targetBone {boneName=spine_02}}
     {targetBone {boneName=spine_03}}
     {targetBone {boneName=spine_04}}
     {targetBone {boneName=spine_05}}
     {targetBone {boneName=pelvis}}
    ]
    constraintWeights [1,1,1,1,1,1]
   }
   n34 = "Constraint - dyn_bk_belt_r_MSTR (PPPP)"!(
    ComponentPose=@n33
    Alpha=1
    ConstraintWeights_0=0.2948
    ConstraintWeights_1:float=0.0336
    ConstraintWeights_2:float=0.0078
    ConstraintWeights_3:float=0.6638
   ) -> (Pose) {
    alphaBoolBlend {blendOption=Linear}
    boneToModify {boneName=dyn_bk_belt_r_MSTR}
    constraintSetup [{targetBone {boneName=spine_01}}, {targetBone {boneName=spine_02}}, {targetBone {boneName=spine_03}}, {targetBone {boneName=pelvis}}]
    constraintWeights [1,1,1,1]
   }
   n35 = "Constraint - dyn_pistol_MSTR (PPP)"!(ComponentPose=@n34, Alpha=1, ConstraintWeights_0=0.0524, ConstraintWeights_1:float=0.4993, ConstraintWeights_2:float=0.4483) -> (Pose) {
    alphaBoolBlend {blendOption=Linear}
    boneToModify {boneName=dyn_pistol_MSTR}
    constraintSetup [{targetBone {boneName=pelvis}}, {targetBone {boneName=thigh_twist_01_l}}, {targetBone {boneName=thigh_l}}]
    constraintWeights [1,1,1]
   }
   n36 = "Constraint - dyn_intercom_MSTR (PPPP)"!(
    ComponentPose=@n35
    Alpha=1
    ConstraintWeights_0=0.3197
    ConstraintWeights_1:float=0.0183
    ConstraintWeights_2:float=0.6366
    ConstraintWeights_3:float=0.0255
   ) -> (Pose) {
    alphaBoolBlend {blendOption=Linear}
    boneToModify {boneName=dyn_intercom_MSTR}
    constraintSetup [
     {targetBone {boneName=clavicle_l}}
     {targetBone {boneName=spine_04}}
     {targetBone {boneName=spine_05}}
     {targetBone {boneName=upperarm_twist_01_l}}
    ]
    constraintWeights [1,1,1,1]
   }
   Save cached pose 'startCache'(Pose=Component To Local(ComponentPose=@n36)) {cacheName=startCache}
   n37 = Get BwdMovement() -> (BwdMovement:double)
   n38 = Look At - Bone: dyn_ft_skirt_01_l(ComponentPose=Local To Component(LocalPose=Use cached pose 'startCache'()), Alpha=1) {
    alphaBoolBlend {blendOption=Linear}
    boneToModify {boneName=dyn_ft_skirt_01_l}
    lookAtLocation=<-15 5 7>
    lookAtTarget {boneReference {boneName=thigh_twist_01_l}}
    lookAt_Axis {axis=<1 0>}
    lookUp_Axis {axis=<0 1>}
   }
   n39 = Look At - Bone: dyn_ft_skirt_02_l(ComponentPose=@n38, Alpha=1) {
    alphaBoolBlend {blendOption=Linear}
    boneToModify {boneName=dyn_ft_skirt_02_l}
    lookAtLocation=<-15 14 7>
    lookAtTarget {boneReference {boneName=thigh_twist_01_l}}
    lookAt_Axis {axis=<1 0>}
    lookUp_Axis {axis=<0 1>}
   }
   n40 = Get FwdMovement() -> (FwdMovement:double)
   n41 = Trail controller - Bone: dyn_ft_skirt_04_l(ComponentPose=@n39, Alpha=@n40) {
    chainLength=3
    debugLifeTime=0.1
    maxDeltaTime=0.0333
    rotationLimits [{limitMax=<180 180 180>}, {limitMax=<180 180 180>}, {limitMax=<180 180 180>}]
    rotationOffsets [<0 0 0>, <0 0 0>, <0 0 0>]
    trailBone {boneName=dyn_ft_skirt_04_l}
   }
   n42 = Get thigh_rot_up_l()
   n43 = "Transform (Modify) Bone - Bone: dyn_ft_skirt_l_MSTR"(ComponentPose=@n41, Alpha=@n42, Translation="1,0,0.000", Rotation="-5,0,0", Scale="0.6,1,1") {
    rotationMode=BMM_Additive
    rotationSpace=BCS_BoneSpace
    scaleMode=BMM_Additive
    scaleSpace=BCS_BoneSpace
    translationMode=BMM_Additive
    translationSpace=BCS_BoneSpace
    alphaBoolBlend {blendOption=Linear}
    boneToModify {boneName=dyn_ft_skirt_l_MSTR}
    rotation {pitch=-5}
    scale {x=0.6}
    translation {x=1}
   }
   n44 = Look At - Bone: dyn_ft_belt_01_l(ComponentPose=@n43, Alpha=1) {
    alphaBoolBlend {blendOption=Linear}
    boneToModify {boneName=dyn_ft_belt_01_l}
    lookAtLocation=<-15 15 7>
    lookAtTarget {boneReference {boneName=thigh_twist_01_l}}
    lookAt_Axis {axis=<1 0>}
    lookUp_Axis {axis=<0 1>}
   }
   n45 = Look At - Bone: dyn_ft_belt_02_l(ComponentPose=@n44, Alpha=1) {
    alphaBoolBlend {blendOption=Linear}
    boneToModify {boneName=dyn_ft_belt_02_l}
    lookAtLocation=<-15 22 7>
    lookAtTarget {boneReference {boneName=thigh_twist_01_l}}
    lookAt_Axis {axis=<1 0>}
    lookUp_Axis {axis=<0 1>}
   }
   n46 = Trail controller - Bone: dyn_ft_belt_04_l(ComponentPose=@n45, Alpha=@n40) {
    chainLength=4
    debugLifeTime=0.1
    maxDeltaTime=0.0333
    rotationLimits [{limitMax=<180 180 180>}, {limitMax=<180 180 180>}, {limitMax=<180 180 180>}, {limitMax=<180 180 180>}]
    rotationOffsets [<0 0 0>, <0 0 0>, <0 0 0>, <0 0 0>]
    trailBone {boneName=dyn_ft_belt_04_l}
   }
   n47 = "Transform (Modify) Bone - Bone: dyn_ft_belt_l_MSTR"(ComponentPose=@n46, Alpha=@n42, Translation="1,0,0.000", Rotation="0,0,0", Scale="0.6,1,1") {
    rotationMode=BMM_Additive
    rotationSpace=BCS_BoneSpace
    scaleMode=BMM_Additive
    scaleSpace=BCS_BoneSpace
    translationMode=BMM_Additive
    translationSpace=BCS_BoneSpace
    alphaBoolBlend {blendOption=Linear}
    boneToModify {boneName=dyn_ft_belt_l_MSTR}
    scale {x=0.6}
    translation {x=1}
   }
   n48 = Look At - Bone: dyn_ft_beltA_01_r(ComponentPose=@n47, Alpha=1) {
    bUseLookUpAxis=true
    alphaBoolBlend {blendOption=Linear}
    boneToModify {boneName=dyn_ft_beltA_01_r}
    lookAtLocation=<15 -20 -13>
    lookAtTarget {boneReference {boneName=thigh_twist_01_r}}
    lookAt_Axis {axis=<1 0>}
    lookUp_Axis {axis=<0 1>}
   }
   n49 = Look At - Bone: dyn_ft_beltA_02_r(ComponentPose=@n48, Alpha=1) {
    bUseLookUpAxis=true
    alphaBoolBlend {blendOption=Linear}
    boneToModify {boneName=dyn_ft_beltA_02_r}
    lookAtLocation=<15 -25 -13>
    lookAtTarget {boneReference {boneName=thigh_twist_01_r}}
    lookAt_Axis {axis=<1 0>}
    lookUp_Axis {axis=<0 1>}
   }
   n50 = Get FwdMovement()
   n51 = Trail controller - Bone: dyn_ft_beltA_03_r(ComponentPose=@n49, Alpha=@n50) {
    chainLength=3
    debugLifeTime=0.1
    maxDeltaTime=0.0333
    rotationLimits [{limitMax=<180 180 180>}, {limitMax=<180 180 180>}, {limitMax=<180 180 180>}]
    rotationOffsets [<0 0 0>, <0 0 0>, <0 0 0>]
    trailBone {boneName=dyn_ft_beltA_03_r}
   }
   n52 = Look At - Bone: dyn_ft_beltB_01_r(ComponentPose=@n51, Alpha=1) {
    bUseLookUpAxis=true
    alphaBoolBlend {blendOption=Linear}
    boneToModify {boneName=dyn_ft_beltB_01_r}
    lookAtLocation=<15 -12 -8>
    lookAtTarget {boneReference {boneName=thigh_twist_01_r}}
    lookAt_Axis {axis=<1 0>}
    lookUp_Axis {axis=<0 1>}
   }
   n53 = Look At - Bone: dyn_ft_beltB_02_r(ComponentPose=@n52, Alpha=1) {
    bUseLookUpAxis=true
    alphaBoolBlend {blendOption=Linear}
    boneToModify {boneName=dyn_ft_beltB_02_r}
    lookAtLocation=<15 -20 -8>
    lookAtTarget {boneReference {boneName=thigh_twist_01_r}}
    lookAt_Axis {axis=<1 0>}
    lookUp_Axis {axis=<0 1>}
   }
   n54 = Trail controller - Bone: dyn_ft_beltB_03_r(ComponentPose=@n53, Alpha=@n50) {
    bUsePlanarLimit=true
    chainLength=3
    debugLifeTime=0.1
    maxDeltaTime=0.0333
    planarLimits [{drivingBone {boneName=spine_01}, planeTransform {location=<0 8 0>, rotation=<rot 0 -5 90>, scale=<0.2 0.5 0.5>}}]
    rotationLimits [{limitMax=<180 180 180>}, {limitMax=<180 180 180>}, {limitMax=<180 180 180>}]
    rotationOffsets [<0 0 0>, <0 0 0>, <0 0 0>]
    trailBone {boneName=dyn_ft_beltB_03_r}
   }
   n55 = Trail controller - Bone: dyn_bk_beltB_04_r(ComponentPose=@n54, Alpha=@n37) {
    bUsePlanarLimit=true
    maxDeltaTime=0.0333
    alphaScaleBiasClamp {bInterpResult=true}
    planarLimits [{drivingBone {boneName=pelvis}, planeTransform {location=<0 -11 0>, rotation=<rot 0 5 -80>, scale=<0.5 0.5 0.5>}}]
    rotationLimits [{limitMax=<180 180 180>}, {limitMax=<180 180 180>}]
    rotationOffsets [<0 0 0>, <0 0 0>]
    trailBone {boneName=dyn_bk_beltB_04_r}
   }
   n56 = Trail controller - Bone: dyn_bk_skirt_03_l(ComponentPose=@n55, Alpha=@n37) {
    bUsePlanarLimit=true
    chainLength=3
    maxDeltaTime=0.0333
    alphaScaleBiasClamp {bInterpResult=true}
    planarLimits [{drivingBone {boneName=pelvis}, planeTransform {location=<0 -11 0>, rotation=<rot 0 -5 -90>, scale=<0.5 0.5 0.5>}}]
    rotationLimits [{limitMax=<180 180 180>}, {limitMax=<180 180 180>}, {limitMax=<180 180 180>}]
    rotationOffsets [<0 0 0>, <0 0 0>, <0 0 0>]
    trailBone {boneName=dyn_bk_skirt_03_l}
    trailRelaxationSpeed {editorCurveData {keys [<at 0 10 RCIM_Linear>,<at 0.5055 10 RCIM_Linear>,<at 1 5 RCIM_Linear>]}}
   }
   n57 = Trail controller - Bone: dyn_bk_skirt_02_r(ComponentPose=@n56, Alpha=@n37) {
    bUsePlanarLimit=true
    maxDeltaTime=0.0333
    alphaScaleBiasClamp {bInterpResult=true}
    planarLimits [{drivingBone {boneName=pelvis}, planeTransform {location=<0 -12 0>, rotation=<rot 0 -5.0001 -89.9999>, scale=<0.5 0.5 0.5>}}]
    rotationLimits [{limitMax=<180 180 180>}, {limitMax=<180 180 180>}]
    rotationOffsets [<0 0 0>, <0 0 0>]
    trailBone {boneName=dyn_bk_skirt_02_r}
    trailRelaxationSpeed {editorCurveData {keys [<at 0 10 RCIM_Linear>,<at 0.5055 10 RCIM_Linear>,<at 1 5 RCIM_Linear>]}}
   }
   n58 = Get bIsSkydiveFloating()
   n59 = "Transform (Modify) Bone - Bone: dyn_bk_skirt_MSTR"!(ComponentPose=@n57, bAlphaBoolEnabled:bool=@n58, Translation="0,0,1", Rotation="5.000,0.000,0.000", Scale="1,1,1") -> (Pose) {
    alphaInputType=Bool
    rotationMode=BMM_Additive
    rotationSpace=BCS_BoneSpace
    scaleMode=BMM_Additive
    scaleSpace=BCS_BoneSpace
    translationMode=BMM_Additive
    translationSpace=BCS_BoneSpace
    alphaBoolBlend {blendOption=Linear}
    boneToModify {boneName=dyn_bk_skirt_MSTR}
    rotation {pitch=5}
    translation {z=1}
   }
   n60 = "Transform (Modify) Bone - Bone: dyn_bk_skirt_r_MSTR"!(ComponentPose=@n59, bAlphaBoolEnabled:bool=@n58, Translation="0,0,1", Rotation="5.000,0.000,0.000", Scale="1,1,1") -> (Pose) {
    alphaInputType=Bool
    rotationMode=BMM_Additive
    rotationSpace=BCS_BoneSpace
    scaleMode=BMM_Additive
    scaleSpace=BCS_BoneSpace
    translationMode=BMM_Additive
    translationSpace=BCS_BoneSpace
    alphaBoolBlend {blendOption=Linear}
    boneToModify {boneName=dyn_bk_skirt_r_MSTR}
    rotation {pitch=5}
    translation {z=1}
   }
   n61 = "Transform (Modify) Bone - Bone: dyn_bk_belt_r_MSTR"(ComponentPose=@n60, bAlphaBoolEnabled=@n58, Translation="0,0,1", Rotation="10.000,0.000,0.000", Scale="1,1,1") {
    alphaInputType=Bool
    rotationMode=BMM_Additive
    rotationSpace=BCS_BoneSpace
    scaleMode=BMM_Additive
    scaleSpace=BCS_BoneSpace
    translationMode=BMM_Additive
    translationSpace=BCS_BoneSpace
    alphaBoolBlend {blendOption=Linear}
    boneToModify {boneName=dyn_bk_belt_r_MSTR}
    rotation {pitch=10}
    translation {z=1}
   }
   Save cached pose 'trailCache'(Pose=Component To Local(ComponentPose=@n61)) {cacheName=trailCache}
  ]}
 EventGraph {nodes [
   Event BlueprintUpdateAnimation()
   > n1 = Sequence!(execute) -> (then_0, then_1, then_2:Exec, then_3:Exec, then_4:Exec)
   n2 = PlayerMovement(execute=@n1.then_2, AnimInstance=Self-Reference(), XDotProductCoefficient=1, YDotProductCoefficient=1, ZDotProductCoefficient=1)
   n3 = MapRangeClamped(Value=@n2.YMovement, InRangeA=0, InRangeB=1)
   > Set FwdMovement(FwdMovement:double=MapRangeClamped(Value=float + float(A=@n2, B=@n3) {is=title K2Node_CommutativeAssociativeBinaryOperator}, InRangeA=0, InRangeB=1, OutRangeA=0.2, OutRangeB=0)) -> (Output_Get:double)
   n4 = MapRangeClamped(Value=@n2.YMovement, InRangeA=0, InRangeB=-1)
   n5 = "Absolute (Float)"(A=@n2.ZMovement)
   > Set BwdMovement(BwdMovement:double=MapRangeClamped(Value=float + float(A=@n5, B=@n4), InRangeA=0, InRangeB=1.5, OutRangeA=0.2, OutRangeB=0)) -> (Output_Get:double)
   n6 = Get bIsSkyDiving()
   Set Skydiving 05(execute=@n1.then_3, Skydiving 05:double=SelectFloat(A=0.5, B=1, bPickA=@n6)) -> (Output_Get:double)
   > Set Skydiving 025(Skydiving 025:double=SelectFloat(A=0.25, B=1, bPickA=@n6)) -> (Output_Get:double)
   n7 = Gravity Override State Function DEPRECATED(
    execute=@n1.then_4
    CharacterPartAnimInstance=Self-Reference()
    GameTimeInSeconds=GetGameTimeInSeconds()
    GravityOverrideParams=Get GravityOverrideParams() -> (GravityOverrideParams:Gravity Override Params Struct Structure)
   )
   > Set GravityOverride(GravityOverride=@n7)
   n8 = BreakTransform(InTransform=GetDeltaTransformFromRefPose(self=GetOwnerSkeletalMeshComponent(), BoneName=thigh_l, BaseName=pelvis))
   Set thigh_rot_up_l(execute=@n1.then_0, thigh_rot_up_l:double=MapRangeClamped(Value=BreakRotator(InRot=@n8.Rotation).Yaw, InRangeA=-30, InRangeB=-100)) -> (Output_Get:double)
   n9 = BreakTransform(InTransform=GetDeltaTransformFromRefPose(self=GetOwnerSkeletalMeshComponent(), BoneName=thigh_r, BaseName=pelvis))
   Set thigh_rot_up_r(execute=@n1.then_1, thigh_rot_up_r:double=MapRangeClamped(Value=BreakRotator(InRot=@n9.Rotation).Yaw, InRangeA=-30, InRangeB=-100)) -> (Output_Get:double)
  ]}
]
variables [
 BwdMovement {type=number}
 FwdMovement {type=number}
 GravityOverride {type=Vector}
 GravityOverrideParams {type=GravityOverrideParamsStruct}
 isFloating {type=number}
 Skydiving 025 {type=number}
 Skydiving 05 {type=number}
 thigh_rot_up_l {type=number}
 thigh_rot_up_r {type=number}
]
