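# Physics asset for the front-end hair of the female Sailor Squad Leader cosmetic body.
# Five capsule chains (left and right, plus back left, middle and right) hang from MSTR bones constrained to the head, with tight per-link angular limits and little or no twist.
# Head, neck, upper spine, clavicle and right upper-arm capsules form the body shapes, including long thin neck capsules and oversized offset spine_05 volumes.
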PhysicsAsset /BRCosmetics/Characters/Player/Female/Medium/Bodies/F_MED_Sailor_Squad_Leader/Meshes/Parts/F_MED_Sailor_Squad_Leader_FrontEnd_hair
bodies {
 clavicle_l [{shapeType=Capsule, length=4.9448, radius=7.8, center=<5.0017 -0.1723 -0.9898>, rotation=<rot 90 0 180>}]
 clavicle_r [{shapeType=Capsule, length=4.9448, radius=8.75, center=<-5.0017 1.944 2.1943>, rotation=<rot 90 0 0>}]
 dyn_hair_01_l [{shapeType=Capsule, length=3, radius=1.5, center=<2.5 0 0>, rotation=<rot 90 0 0>}]
 dyn_hair_01_r [{shapeType=Capsule, length=3.125, radius=1.125, center=<2 0 0>, rotation=<rot 90 0 0>}]
 dyn_hair_02_l [{shapeType=Capsule, length=2.75, radius=1.75, center=<2.5 0 0>, rotation=<rot 90 0 0>}]
 dyn_hair_02_r [{shapeType=Capsule, length=2.875, radius=1.125, center=<2 0 0>, rotation=<rot 90 0 0>}]
 dyn_hair_03_l [{shapeType=Capsule, length=6.5, radius=1.6, center=<4.2653 0.4142 0>, rotation=<rot 90 73.0658 43.0655>}]
 dyn_hair_03_r [{shapeType=Capsule, length=4.6335, radius=1.75, center=<5.5179 -0.1443 0>, rotation=<rot 90 0 5>}]
 dyn_hair_l_01_bk [{shapeType=Capsule, length=2.75, radius=2, center=<3 0 0>, rotation=<rot 90 0 0>}]
 dyn_hair_l_02_bk [{shapeType=Capsule, length=2.5, radius=2.5, center=<3 0 0>, rotation=<rot 90 0 0>}]
 dyn_hair_l_03_bk [{shapeType=Capsule, length=3, radius=1.75, center=<3.7534 0 0.4322>, rotation=<rot 87.2754 0 0>}]
 dyn_hair_l_04_bk [{shapeType=Capsule, length=3.5, radius=1.25, center=<3 0 0>, rotation=<rot 90 0 0>}]
 dyn_hair_l_MSTR [{shapeType=Capsule, length=0.505, radius=0.505, rotation=<rot 0 -90 0>}]
 dyn_hair_l_bk_MSTR [{shapeType=Capsule, length=0.505, radius=0.505, rotation=<rot 0 -90 0>}]
 dyn_hair_mid_01_bk [{shapeType=Capsule, length=6.75, radius=2, center=<3 0 0>, rotation=<rot 0 0 -90>}]
 dyn_hair_mid_02_bk [{shapeType=Capsule, length=12, radius=2, center=<3 0 -0.2>, rotation=<rot 0 0 90>}]
 dyn_hair_mid_03_bk [{shapeType=Capsule, length=12.25, radius=2, shapeName=dyn_hair_mid_03_bk_capsule, center=<3 0 0>, rotation=<rot 0 0 -90>}]
 dyn_hair_mid_04_bk [{shapeType=Capsule, length=13.25, radius=1.5, center=<3 0 0.0461>, rotation=<rot 0 0 90>}]
 dyn_hair_mid_05_bk [{shapeType=Capsule, length=5.5, radius=1.25, center=<4 0 0>, rotation=<rot 0 0 -90>}]
 dyn_hair_mid_bk_MSTR [{shapeType=Capsule, length=0.505, radius=0.505, rotation=<rot 0 -90 0>}]
 dyn_hair_r_01_bk [{shapeType=Capsule, length=1.875, radius=1.5, center=<3 0 0>, rotation=<rot 90 0 0>}]
 dyn_hair_r_02_bk [{shapeType=Capsule, length=3, radius=1.5, center=<3.3574 0 -0.5>, rotation=<rot 90 0 0>}]
 dyn_hair_r_03_bk [{shapeType=Capsule, length=2.8804, radius=1.5, center=<4 0 0.6>, rotation=<rot 90 0 0>}]
 dyn_hair_r_04_bk [{shapeType=Capsule, length=3.5, radius=1.5, center=<3.7 -0.41 0.15>, rotation=<rot 85 0 0>}]
 dyn_hair_r_MSTR [{shapeType=Capsule, length=0.505, radius=0.505, rotation=<rot 0 -90 0>}]
 dyn_hair_r_bk_MSTR [{shapeType=Capsule, length=0.505, radius=0.505, rotation=<rot 0 -90 0>}]
 head [{shapeType=Capsule, length=10.1948, radius=5.7069, center=<0 0.9264 0>, rotation=<rot 90 0 0>}]
 neck_01 [
  {shapeType=Capsule, length=0.75, radius=5.7069, center=<2.0842 -0.3521 0>, rotation=<rot 90 0 0>}
  {shapeType=Capsule, length=65.25, radius=1.25, shapeName=neck_01_capsule, center=<6.0523 -0.5511 0.065>, rotation=<rot 0 -0.0001 5>}
  {shapeType=Capsule, length=65.75, radius=0.75, shapeName=neck_01_capsule, center=<2.2883 0.9529 0>, rotation=<rot 0 -0.0001 5>}
  {shapeType=Capsule, length=65.25, radius=1.25, shapeName=neck_01_capsule, center=<10.7728 -0.824 0.1555>, rotation=<rot 0 -0.0001 5>}
 ]
 spine_04 [{shapeType=Capsule, length=11.0196, radius=10.5135, center=<1.7556 -1.8482 0>}]
 spine_05 [
  {shapeType=Capsule, length=17.05, radius=8.75, center=<3.9714 -1.5339 0>}
  {shapeType=Capsule, length=10, radius=18, shapeName=spine_05_capsule, center=<10.0867 -31.4067 1.9279>}
  {shapeType=Capsule, length=16.05, radius=8.75, center=<0.9379 1.0224 0>}
  {shapeType=Capsule, length=15, radius=15, shapeName=spine_05_capsule, center=<31.7686 -20.0973 0>}
 ]
 upperarm_r [{shapeType=Capsule, length=18.0698, radius=3.9569, center=<-9.1234 0.3191 1.7882>, rotation=<rot 90 0 0>}]
}
constraints [
 {
  bone1Name=dyn_hair_l_MSTR
  bone2Name=head
  swing1LimitDegrees=45
  swing1Motion=Free
  swing2LimitDegrees=45
  swing2Motion=Free
  twistLimitDegrees=45
  twistMotion=Free
 }
 {bone1Name=dyn_hair_01_l, bone2Name=dyn_hair_l_MSTR, swing1LimitDegrees=10, swing2LimitDegrees=10, twistLimitDegrees=5}
 {bone1Name=dyn_hair_02_l, bone2Name=dyn_hair_01_l, swing1LimitDegrees=15, swing2LimitDegrees=30, twistLimitDegrees=5}
 {bone1Name=dyn_hair_03_l, bone2Name=dyn_hair_02_l, swing1LimitDegrees=15, swing2LimitDegrees=30, twistLimitDegrees=5}
 {
  bone1Name=dyn_hair_r_MSTR
  bone2Name=head
  swing1LimitDegrees=45
  swing1Motion=Free
  swing2LimitDegrees=45
  swing2Motion=Free
  twistLimitDegrees=45
  twistMotion=Free
 }
 {bone1Name=dyn_hair_01_r, bone2Name=dyn_hair_r_MSTR, swing1LimitDegrees=10, swing2LimitDegrees=10, twistLimitDegrees=0}
 {bone1Name=dyn_hair_02_r, bone2Name=dyn_hair_01_r, swing1LimitDegrees=10, swing2LimitDegrees=10, twistLimitDegrees=0}
 {bone1Name=dyn_hair_03_r, bone2Name=dyn_hair_02_r, swing1LimitDegrees=10, swing2LimitDegrees=10, twistLimitDegrees=0}
 {
  bone1Name=dyn_hair_r_bk_MSTR
  bone2Name=head
  swing1LimitDegrees=45
  swing1Motion=Free
  swing2LimitDegrees=45
  swing2Motion=Free
  twistLimitDegrees=45
  twistMotion=Free
 }
 {bone1Name=dyn_hair_r_01_bk, bone2Name=dyn_hair_r_bk_MSTR, swing1LimitDegrees=10, swing2LimitDegrees=20, twistLimitDegrees=0}
 {bone1Name=dyn_hair_r_02_bk, bone2Name=dyn_hair_r_01_bk, swing1LimitDegrees=5, swing2LimitDegrees=20, twistLimitDegrees=0}
 {bone1Name=dyn_hair_r_03_bk, bone2Name=dyn_hair_r_02_bk, swing1LimitDegrees=1, swing2LimitDegrees=20, twistLimitDegrees=0}
 {bone1Name=dyn_hair_r_04_bk, bone2Name=dyn_hair_r_03_bk, swing1LimitDegrees=1, swing2LimitDegrees=20, twistLimitDegrees=0}
 {
  bone1Name=dyn_hair_mid_bk_MSTR
  bone2Name=head
  swing1LimitDegrees=45
  swing1Motion=Free
  swing2LimitDegrees=45
  swing2Motion=Free
  twistLimitDegrees=45
  twistMotion=Free
 }
 {bone1Name=dyn_hair_mid_01_bk, bone2Name=dyn_hair_mid_bk_MSTR, swing1LimitDegrees=5, swing2LimitDegrees=40, twistLimitDegrees=0}
 {bone1Name=dyn_hair_mid_02_bk, bone2Name=dyn_hair_mid_01_bk, swing1LimitDegrees=1, swing2LimitDegrees=20, twistLimitDegrees=0}
 {bone1Name=dyn_hair_mid_03_bk, bone2Name=dyn_hair_mid_02_bk, swing1LimitDegrees=1, swing2LimitDegrees=35, twistLimitDegrees=0}
 {bone1Name=dyn_hair_mid_04_bk, bone2Name=dyn_hair_mid_03_bk, swing1LimitDegrees=1, swing2LimitDegrees=20, twistLimitDegrees=0}
 {bone1Name=dyn_hair_mid_05_bk, bone2Name=dyn_hair_mid_04_bk, swing1LimitDegrees=1, swing2LimitDegrees=20, twistLimitDegrees=0}
 {
  bone1Name=dyn_hair_l_bk_MSTR
  bone2Name=head
  swing1LimitDegrees=45
  swing1Motion=Free
  swing2LimitDegrees=45
  swing2Motion=Free
  twistLimitDegrees=45
  twistMotion=Free
 }
 {bone1Name=dyn_hair_l_01_bk, bone2Name=dyn_hair_l_bk_MSTR, swing1LimitDegrees=15, swing2LimitDegrees=40, twistLimitDegrees=0}
 {bone1Name=dyn_hair_l_02_bk, bone2Name=dyn_hair_l_01_bk, swing1LimitDegrees=15, swing2LimitDegrees=20, twistLimitDegrees=0}
 {bone1Name=dyn_hair_l_03_bk, bone2Name=dyn_hair_l_02_bk, swing1LimitDegrees=15, swing2LimitDegrees=35, twistLimitDegrees=0}
 {bone1Name=dyn_hair_l_04_bk, bone2Name=dyn_hair_l_03_bk, swing1LimitDegrees=15, swing2LimitDegrees=20, twistLimitDegrees=0}
 {
  bone1Name=head
  bone2Name=neck_01
  swing1LimitDegrees=45
  swing1Motion=Free
  swing2LimitDegrees=45
  swing2Motion=Free
  twistLimitDegrees=45
  twistMotion=Free
 }
 {
  bone1Name=clavicle_r
  bone2Name=spine_05
  swing1LimitDegrees=45
  swing1Motion=Free
  swing2LimitDegrees=45
  swing2Motion=Free
  twistLimitDegrees=45
  twistMotion=Free
 }
 {
  bone1Name=neck_01
  bone2Name=spine_05
  swing1LimitDegrees=45
  swing1Motion=Free
  swing2LimitDegrees=45
  swing2Motion=Free
  twistLimitDegrees=45
  twistMotion=Free
 }
 {
  bone1Name=clavicle_l
  bone2Name=spine_05
  swing1LimitDegrees=45
  swing1Motion=Free
  swing2LimitDegrees=45
  swing2Motion=Free
  twistLimitDegrees=45
  twistMotion=Free
 }
 {
  bone1Name=spine_05
  bone2Name=spine_04
  swing1LimitDegrees=45
  swing1Motion=Free
  swing2LimitDegrees=45
  swing2Motion=Free
  twistLimitDegrees=45
  twistMotion=Free
 }
 {
  bone1Name=upperarm_r
  bone2Name=clavicle_r
  swing1LimitDegrees=45
  swing1Motion=Free
  swing2LimitDegrees=45
  swing2Motion=Free
  twistLimitDegrees=45
  twistMotion=Free
 }
]
