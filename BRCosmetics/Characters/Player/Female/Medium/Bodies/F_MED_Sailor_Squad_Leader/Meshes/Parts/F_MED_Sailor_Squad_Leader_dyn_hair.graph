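# Physics asset for the dynamic back hair of the female medium Sailor Squad Leader body part.
# Left, middle and right capsule chains hang from small MSTR root bodies, which link to the head with free angular joints.
# Links within each chain have tight swing limits, with zero twist on the side chains and some twist along the middle one.
# Head, neck, upper spine, clavicle and right upper-arm capsules supply body volume, including oversized offset capsules on head and spine_05.
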PhysicsAsset /BRCosmetics/Characters/Player/Female/Medium/Bodies/F_MED_Sailor_Squad_Leader/Meshes/Parts/F_MED_Sailor_Squad_Leader_dyn_hair
bodies {
 clavicle_l [{shapeType=Capsule, length=4.9984, radius=7.9034, center=<5.0017 -0.536 -1.5153>, rotation=<rot 80 -0.0003 179.9997>}]
 clavicle_r [{shapeType=Capsule, length=3.6327, radius=9.5621, center=<-4.4387 0.536 3.5284>, rotation=<rot 80 -0.0005 -0.0005>}]
 dyn_hair_l_01_bk [{shapeType=Capsule, length=2.75, radius=2, center=<3 0 0>, rotation=<rot 90 0 0>}]
 dyn_hair_l_02_bk [{shapeType=Capsule, length=2.2976, radius=2.3147, center=<3 0 1.129>, rotation=<rot 90 0 0>}]
 dyn_hair_l_03_bk [{shapeType=Capsule, length=3, radius=2.6362, center=<3.808 0 1.58>, rotation=<rot 87.2754 0 0>}]
 dyn_hair_l_04_bk [{shapeType=Capsule, length=3.5, radius=2.4209, center=<2.9903 0 1.326>, rotation=<rot 84.9999 0.0001 0.0001>}]
 dyn_hair_l_bk_MSTR [{shapeType=Capsule, length=0.505, radius=0.505, rotation=<rot 0 -90 0>}]
 dyn_hair_mid_01_bk [{shapeType=Capsule, length=6.75, radius=1.5, center=<3 0 0>, rotation=<rot 0 0 -90>}]
 dyn_hair_mid_02_bk [{shapeType=Capsule, length=2, radius=1.7, center=<3 0 -0.2>, rotation=<rot 90 0 0>}]
 dyn_hair_mid_03_bk [{shapeType=Capsule, length=1.6014, radius=1.5, shapeName=dyn_hair_mid_03_bk_capsule, center=<3 0 0>, rotation=<rot 90 0 0>}]
 dyn_hair_mid_04_bk [{shapeType=Capsule, length=4.0312, radius=1.4, center=<2.706 0 0.0461>, rotation=<rot 0 -90 90.0001>}]
 dyn_hair_mid_05_bk [{shapeType=Capsule, length=5.5, radius=1.2, center=<4 0 0>, rotation=<rot 90 0 0>}]
 dyn_hair_mid_bk_MSTR [{shapeType=Capsule, length=0.505, radius=0.505, rotation=<rot 0 -90 0>}]
 dyn_hair_r_01_bk [{shapeType=Capsule, length=1.875, radius=1.5, center=<3 0 0>, rotation=<rot 90 0 0>}]
 dyn_hair_r_02_bk [{shapeType=Capsule, length=2.3367, radius=2.1633, center=<3.3126 0 0>, rotation=<rot 80 180 0>}]
 dyn_hair_r_03_bk [{shapeType=Capsule, length=2.5218, radius=2.823, center=<4 0 1.9348>, rotation=<rot 90 0 0>}]
 dyn_hair_r_04_bk [{shapeType=Capsule, length=3.5, radius=2.4031, center=<3.8493 -0.41 1.8564>, rotation=<rot 85 0 0>}]
 dyn_hair_r_bk_MSTR [{shapeType=Capsule, length=0.505, radius=0.505, rotation=<rot 0 -90 0>}]
 head [
  {shapeType=Capsule, length=5.1948, radius=5.4569, center=<0 0.9264 0>, rotation=<rot 90 0 0>}
  {shapeType=Capsule, length=50.051, radius=7.877, shapeName=head_capsule, center=<-10.9086 4.7768 0>}
 ]
 neck_01 [{shapeType=Capsule, length=2.6948, radius=4.9569, center=<0.8347 -0.3521 0>, rotation=<rot 90 0 0>}]
 spine_04 [{shapeType=Capsule, length=11.5196, radius=10.0135, center=<3.899 -1.3699 0>}]
 spine_05 [
  {shapeType=Capsule, length=17.2988, radius=9.6176, center=<1.8001 -2.1368 0>}
  {shapeType=Capsule, length=26.7035, radius=15, shapeName=spine_05_capsule, center=<10.7796 -30.085 1.9279>}
  {shapeType=Capsule, length=26.7035, radius=15, shapeName=spine_05_capsule, center=<25.1797 -28.0665 1.9279>}
  {shapeType=Capsule, length=57.6686, radius=15, shapeName=spine_05_capsule, center=<11.5129 18.1019 5.1145>, rotation=<rot 0 -0.0002 -20>}
 ]
 upperarm_r [{shapeType=Capsule, length=19, radius=4.5, center=<-13.4169 0.3191 1.7882>, rotation=<rot 90 0 0>}]
}
constraints [
 {
  bone1Name=dyn_hair_r_bk_MSTR
  bone2Name=head
  swing1LimitDegrees=45
  swing1Motion=Free
  swing2LimitDegrees=45
  swing2Motion=Free
  twistLimitDegrees=45
  twistMotion=Free
 }
 {bone1Name=dyn_hair_r_01_bk, bone2Name=dyn_hair_r_bk_MSTR, swing1LimitDegrees=5, swing2LimitDegrees=20, twistLimitDegrees=0}
 {bone1Name=dyn_hair_r_02_bk, bone2Name=dyn_hair_r_01_bk, swing1LimitDegrees=10, swing2LimitDegrees=10, twistLimitDegrees=0}
 {bone1Name=dyn_hair_r_03_bk, bone2Name=dyn_hair_r_02_bk, swing1LimitDegrees=10, swing2LimitDegrees=15, twistLimitDegrees=0}
 {bone1Name=dyn_hair_r_04_bk, bone2Name=dyn_hair_r_03_bk, swing1LimitDegrees=10, swing2LimitDegrees=8, twistLimitDegrees=0}
 {
  bone1Name=dyn_hair_mid_bk_MSTR
  bone2Name=head
  swing1LimitDegrees=45
  swing1Motion=Free
  swing2LimitDegrees=45
  swing2Motion=Free
  twistLimitDegrees=45
  twistMotion=Free
 }
 {bone1Name=dyn_hair_mid_01_bk, bone2Name=dyn_hair_mid_bk_MSTR, swing1LimitDegrees=2, swing2LimitDegrees=30, twistLimitDegrees=0}
 {bone1Name=dyn_hair_mid_02_bk, bone2Name=dyn_hair_mid_01_bk, swing1LimitDegrees=3, swing2LimitDegrees=30, twistLimitDegrees=20}
 {bone1Name=dyn_hair_mid_03_bk, bone2Name=dyn_hair_mid_02_bk, swing1LimitDegrees=3, swing2LimitDegrees=20, twistLimitDegrees=10}
 {bone1Name=dyn_hair_mid_04_bk, bone2Name=dyn_hair_mid_03_bk, swing1LimitDegrees=3, swing2LimitDegrees=20, twistLimitDegrees=20}
 {bone1Name=dyn_hair_mid_05_bk, bone2Name=dyn_hair_mid_04_bk, swing1LimitDegrees=8, swing2LimitDegrees=5, twistLimitDegrees=30}
 {
  bone1Name=dyn_hair_l_bk_MSTR
  bone2Name=head
  swing1LimitDegrees=45
  swing1Motion=Free
  swing2LimitDegrees=45
  swing2Motion=Free
  twistLimitDegrees=45
  twistMotion=Free
 }
 {bone1Name=dyn_hair_l_01_bk, bone2Name=dyn_hair_l_bk_MSTR, swing1LimitDegrees=2, swing2LimitDegrees=30, twistLimitDegrees=0}
 {bone1Name=dyn_hair_l_02_bk, bone2Name=dyn_hair_l_01_bk, swing1LimitDegrees=2, swing2LimitDegrees=30, twistLimitDegrees=0}
 {bone1Name=dyn_hair_l_03_bk, bone2Name=dyn_hair_l_02_bk, swing1LimitDegrees=8, swing2LimitDegrees=5, twistLimitDegrees=0}
 {bone1Name=dyn_hair_l_04_bk, bone2Name=dyn_hair_l_03_bk, swing1LimitDegrees=8, swing2LimitDegrees=8, twistLimitDegrees=0}
 {
  bone1Name=head
  bone2Name=neck_01
  swing1LimitDegrees=45
  swing1Motion=Free
  swing2LimitDegrees=45
  swing2Motion=Free
  twistLimitDegrees=45
  twistMotion=Free
 }
 {
  bone1Name=clavicle_r
  bone2Name=spine_05
  swing1LimitDegrees=45
  swing1Motion=Free
  swing2LimitDegrees=45
  swing2Motion=Free
  twistLimitDegrees=45
  twistMotion=Free
 }
 {
  bone1Name=neck_01
  bone2Name=spine_05
  swing1LimitDegrees=45
  swing1Motion=Free
  swing2LimitDegrees=45
  swing2Motion=Free
  twistLimitDegrees=45
  twistMotion=Free
 }
 {
  bone1Name=clavicle_l
  bone2Name=spine_05
  swing1LimitDegrees=45
  swing1Motion=Free
  swing2LimitDegrees=45
  swing2Motion=Free
  twistLimitDegrees=45
  twistMotion=Free
 }
 {
  bone1Name=spine_05
  bone2Name=spine_04
  swing1LimitDegrees=45
  swing1Motion=Free
  swing2LimitDegrees=45
  swing2Motion=Free
  twistLimitDegrees=45
  twistMotion=Free
 }
 {
  bone1Name=upperarm_r
  bone2Name=clavicle_r
  swing1LimitDegrees=45
  swing1Motion=Free
  swing2LimitDegrees=45
  swing2Motion=Free
  twistLimitDegrees=45
  twistMotion=Free
 }
]
