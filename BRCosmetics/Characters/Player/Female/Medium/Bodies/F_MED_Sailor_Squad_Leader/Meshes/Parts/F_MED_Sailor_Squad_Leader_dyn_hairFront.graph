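# Physics asset for the dynamic front hair of the female medium Sailor Squad Leader character.
# Left and right three-bone capsule chains hang from small master bodies linked to the head, with tight swing and twist limits along each chain.
# Capsules on the head, neck, upper spine, clavicles and right upper arm (no left) provide body shapes, joined by free-rotation constraints like those at the hair roots.
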PhysicsAsset /BRCosmetics/Characters/Player/Female/Medium/Bodies/F_MED_Sailor_Squad_Leader/Meshes/Parts/F_MED_Sailor_Squad_Leader_dyn_hairFront
bodies {
 clavicle_l [{shapeType=Capsule, length=5.9448, radius=6.9569, center=<5.0017 -1.944 -2.1943>, rotation=<rot 90 0 180>}]
 clavicle_r [{shapeType=Capsule, length=5.6948, radius=7.5, center=<-5.0017 1.944 2.1943>, rotation=<rot 90 0 0>}]
 dyn_hair_01_l [{shapeType=Capsule, length=3, radius=1.5, center=<2.5 0 0>, rotation=<rot 90 0 0>}]
 dyn_hair_01_r [{shapeType=Capsule, length=3.125, radius=1.125, center=<2 0 0>, rotation=<rot 90 0 0>}]
 dyn_hair_02_l [{shapeType=Capsule, length=2.75, radius=1.75, center=<2.5 0 0>, rotation=<rot 90 0 0>}]
 dyn_hair_02_r [{shapeType=Capsule, length=2.875, radius=1.125, center=<2 0 0>, rotation=<rot 90 0 0>}]
 dyn_hair_03_l [{shapeType=Capsule, length=4.75, radius=1.25, center=<4.1713 0.577 0>, rotation=<rot 90 178.2101 168.21>}]
 dyn_hair_03_r [{shapeType=Capsule, length=4.6335, radius=1.75, center=<5.5179 -0.1443 0>, rotation=<rot 90 0 5>}]
 dyn_hair_l_MSTR [{shapeType=Capsule, length=0.505, radius=0.505, rotation=<rot 0 -90 0>}]
 dyn_hair_r_MSTR [{shapeType=Capsule, length=0.505, radius=0.505, rotation=<rot 0 -90 0>}]
 head [{shapeType=Capsule, length=5.1948, radius=5.4569, center=<0 0.9264 0>, rotation=<rot 90 0 0>}]
 neck_01 [{shapeType=Capsule, length=2.6948, radius=4.9569, center=<0.8347 -0.3521 0>, rotation=<rot 90 0 0>}]
 spine_04 [{shapeType=Capsule, length=11.5196, radius=10.0135, center=<3.899 -1.3699 0>}]
 upperarm_r [{shapeType=Capsule, length=19, radius=4.5, center=<-13.4169 0.3191 1.7882>, rotation=<rot 90 0 0>}]
}
constraints [
 {
  bone1Name=dyn_hair_l_MSTR
  bone2Name=head
  swing1LimitDegrees=45
  swing1Motion=Free
  swing2LimitDegrees=45
  swing2Motion=Free
  twistLimitDegrees=45
  twistMotion=Free
 }
 {bone1Name=dyn_hair_01_l, bone2Name=dyn_hair_l_MSTR, swing1LimitDegrees=15, swing2LimitDegrees=30, twistLimitDegrees=5}
 {bone1Name=dyn_hair_02_l, bone2Name=dyn_hair_01_l, swing1LimitDegrees=15, swing2LimitDegrees=30, twistLimitDegrees=5}
 {bone1Name=dyn_hair_03_l, bone2Name=dyn_hair_02_l, swing1LimitDegrees=15, swing2LimitDegrees=30, twistLimitDegrees=5}
 {
  bone1Name=dyn_hair_r_MSTR
  bone2Name=head
  swing1LimitDegrees=45
  swing1Motion=Free
  swing2LimitDegrees=45
  swing2Motion=Free
  twistLimitDegrees=45
  twistMotion=Free
 }
 {bone1Name=dyn_hair_01_r, bone2Name=dyn_hair_r_MSTR, swing1LimitDegrees=15, swing2LimitDegrees=30, twistLimitDegrees=2}
 {bone1Name=dyn_hair_02_r, bone2Name=dyn_hair_01_r, swing1LimitDegrees=15, swing2LimitDegrees=30, twistLimitDegrees=2}
 {bone1Name=dyn_hair_03_r, bone2Name=dyn_hair_02_r, swing1LimitDegrees=15, swing2LimitDegrees=30, twistLimitDegrees=2}
 {
  bone1Name=head
  bone2Name=neck_01
  swing1LimitDegrees=45
  swing1Motion=Free
  swing2LimitDegrees=45
  swing2Motion=Free
  twistLimitDegrees=45
  twistMotion=Free
 }
 {
  bone1Name=clavicle_r
  bone2Name=spine_04
  swing1LimitDegrees=45
  swing1Motion=Free
  swing2LimitDegrees=45
  swing2Motion=Free
  twistLimitDegrees=45
  twistMotion=Free
 }
 {
  bone1Name=neck_01
  bone2Name=spine_04
  swing1LimitDegrees=45
  swing1Motion=Free
  swing2LimitDegrees=45
  swing2Motion=Free
  twistLimitDegrees=45
  twistMotion=Free
 }
 {
  bone1Name=clavicle_l
  bone2Name=spine_04
  swing1LimitDegrees=45
  swing1Motion=Free
  swing2LimitDegrees=45
  swing2Motion=Free
  twistLimitDegrees=45
  twistMotion=Free
 }
 {
  bone1Name=upperarm_r
  bone2Name=clavicle_r
  swing1LimitDegrees=45
  swing1Motion=Free
  swing2LimitDegrees=45
  swing2Motion=Free
  twistLimitDegrees=45
  twistMotion=Free
 }
]
