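# Physics asset for the dynamic back hair of the female medium Sailor Squad Leader character, named as the variant without a backpack.
# Three capsule chains (left, middle, right) hang from small master bodies on the head, with limited swing that mostly widens toward the tips and little or no twist.
# Capsules on head, neck, upper spine, clavicles and right upper arm form the body shapes, including a large offset second capsule on spine_05.
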PhysicsAsset /BRCosmetics/Characters/Player/Female/Medium/Bodies/F_MED_Sailor_Squad_Leader/Meshes/Parts/F_MED_Sailor_Squad_Leader_dyn_hair_WOBackPack
bodies {
 clavicle_l [{shapeType=Capsule, length=5.9448, radius=6.9569, center=<5.0017 -1.944 -2.1943>, rotation=<rot 90 0 180>}]
 clavicle_r [{shapeType=Capsule, length=5.6948, radius=8, center=<-4.4007 1.944 1.5017>, rotation=<rot 80 -0.0002 -0.0003>}]
 dyn_hair_l_01_bk [{shapeType=Capsule, length=2.75, radius=2, center=<3 0 0>, rotation=<rot 90 0 0>}]
 dyn_hair_l_02_bk [{shapeType=Capsule, length=2.5, radius=2.5, center=<3 0 0>, rotation=<rot 90 0 0>}]
 dyn_hair_l_03_bk [{shapeType=Capsule, length=3, radius=1.75, center=<3.7534 0 0.4322>, rotation=<rot 87.2754 0 0>}]
 dyn_hair_l_04_bk [{shapeType=Capsule, length=3.5, radius=1.25, center=<3 0 0>, rotation=<rot 90 0 0>}]
 dyn_hair_l_bk_MSTR [{shapeType=Capsule, length=0.505, radius=0.505, rotation=<rot 0 -90 0>}]
 dyn_hair_mid_01_bk [{shapeType=Capsule, length=2.6126, radius=1.5, center=<3.6062 0 0>, rotation=<rot 90 180 -180>}]
 dyn_hair_mid_02_bk [{shapeType=Capsule, length=3.7528, radius=2, center=<3 0 -0.5>, rotation=<rot 90 0 0>}]
 dyn_hair_mid_03_bk [{shapeType=Capsule, length=2.2659, radius=2, shapeName=dyn_hair_mid_03_bk_capsule, center=<3 0 0>, rotation=<rot 90 0 0>}]
 dyn_hair_mid_04_bk [{shapeType=Capsule, length=3.6991, radius=2, center=<3 0 0.0461>, rotation=<rot 90 0 0>}]
 dyn_hair_mid_05_bk [{shapeType=Capsule, length=5.5, radius=2, center=<4 0 0>, rotation=<rot 0 0 -90>}]
 dyn_hair_mid_bk_MSTR [{shapeType=Capsule, length=0.505, radius=0.505, rotation=<rot 0 -90 0>}]
 dyn_hair_r_01_bk [{shapeType=Capsule, length=1.875, radius=1.5, center=<3 0 0>, rotation=<rot 90 0 0>}]
 dyn_hair_r_02_bk [{shapeType=Capsule, length=3.2392, radius=2.69, center=<3.1491 0 -1.0869>, rotation=<rot 60.0001 179.9999 180>}]
 dyn_hair_r_03_bk [{shapeType=Capsule, length=4.1304, radius=2.2221, center=<4.7782 0 1.7>, rotation=<rot 90 0 0>}]
 dyn_hair_r_04_bk [{shapeType=Capsule, length=3.75, radius=2.4578, center=<4.0463 -0.41 2>, rotation=<rot 85 0 0>}]
 dyn_hair_r_bk_MSTR [{shapeType=Capsule, length=0.505, radius=0.505, rotation=<rot 0 -90 0>}]
 head [{shapeType=Capsule, length=5.1948, radius=5.4569, center=<0 0.9264 0>, rotation=<rot 90 0 0>}]
 neck_01 [{shapeType=Capsule, length=2.6948, radius=4.9569, center=<0.8347 -0.3521 0>, rotation=<rot 90 0 0>}]
 spine_04 [{shapeType=Capsule, length=11.5196, radius=10.0135, center=<3.899 -1.3699 0>}]
 spine_05 [
  {shapeType=Capsule, length=20.0101, radius=9.5, center=<2.6698 -1.8233 0>}
  {shapeType=Capsule, length=36.986, radius=15, shapeName=spine_05_capsule, center=<12.6056 14.7923 1.2834>, rotation=<rot 0 -0.0004 10>}
 ]
 upperarm_r [{shapeType=Capsule, length=18.0698, radius=3.9569, center=<-9.1234 0.3191 1.7882>, rotation=<rot 90 0 0>}]
}
constraints [
 {
  bone1Name=dyn_hair_r_bk_MSTR
  bone2Name=head
  swing1LimitDegrees=45
  swing1Motion=Free
  swing2LimitDegrees=45
  swing2Motion=Free
  twistLimitDegrees=45
  twistMotion=Free
 }
 {bone1Name=dyn_hair_r_01_bk, bone2Name=dyn_hair_r_bk_MSTR, swing1LimitDegrees=10, swing2LimitDegrees=15, twistLimitDegrees=1}
 {bone1Name=dyn_hair_r_02_bk, bone2Name=dyn_hair_r_01_bk, swing1LimitDegrees=10, swing2LimitDegrees=40, twistLimitDegrees=0}
 {bone1Name=dyn_hair_r_03_bk, bone2Name=dyn_hair_r_02_bk, swing1LimitDegrees=10, swing2LimitDegrees=45, twistLimitDegrees=1}
 {bone1Name=dyn_hair_r_04_bk, bone2Name=dyn_hair_r_03_bk, swing1LimitDegrees=15, swing2LimitDegrees=45, twistLimitDegrees=1}
 {
  bone1Name=dyn_hair_mid_bk_MSTR
  bone2Name=head
  swing1LimitDegrees=45
  swing1Motion=Free
  swing2LimitDegrees=45
  swing2Motion=Free
  twistLimitDegrees=45
  twistMotion=Free
 }
 {bone1Name=dyn_hair_mid_01_bk, bone2Name=dyn_hair_mid_bk_MSTR, swing1LimitDegrees=0, swing2LimitDegrees=15, twistLimitDegrees=0}
 {bone1Name=dyn_hair_mid_02_bk, bone2Name=dyn_hair_mid_01_bk, swing1LimitDegrees=5, swing2LimitDegrees=35, twistLimitDegrees=0}
 {bone1Name=dyn_hair_mid_03_bk, bone2Name=dyn_hair_mid_02_bk, swing1LimitDegrees=10, swing2LimitDegrees=35, twistLimitDegrees=0}
 {bone1Name=dyn_hair_mid_04_bk, bone2Name=dyn_hair_mid_03_bk, swing1LimitDegrees=15, swing2LimitDegrees=40, twistLimitDegrees=0}
 {bone1Name=dyn_hair_mid_05_bk, bone2Name=dyn_hair_mid_04_bk, swing1LimitDegrees=20, swing2LimitDegrees=45, twistLimitDegrees=0}
 {
  bone1Name=dyn_hair_l_bk_MSTR
  bone2Name=head
  swing1LimitDegrees=45
  swing1Motion=Free
  swing2LimitDegrees=45
  swing2Motion=Free
  twistLimitDegrees=45
  twistMotion=Free
 }
 {bone1Name=dyn_hair_l_01_bk, bone2Name=dyn_hair_l_bk_MSTR, swing1LimitDegrees=5, swing2LimitDegrees=15, twistLimitDegrees=1}
 {bone1Name=dyn_hair_l_02_bk, bone2Name=dyn_hair_l_01_bk, swing1LimitDegrees=20, swing2LimitDegrees=25, twistLimitDegrees=0}
 {bone1Name=dyn_hair_l_03_bk, bone2Name=dyn_hair_l_02_bk, swing1LimitDegrees=30, swing2LimitDegrees=25, twistLimitDegrees=5}
 {bone1Name=dyn_hair_l_04_bk, bone2Name=dyn_hair_l_03_bk, swing1LimitDegrees=30, swing2LimitDegrees=25, twistLimitDegrees=5}
 {
  bone1Name=head
  bone2Name=neck_01
  swing1LimitDegrees=45
  swing1Motion=Free
  swing2LimitDegrees=45
  swing2Motion=Free
  twistLimitDegrees=45
  twistMotion=Free
 }
 {
  bone1Name=clavicle_r
  bone2Name=spine_05
  swing1LimitDegrees=45
  swing1Motion=Free
  swing2LimitDegrees=45
  swing2Motion=Free
  twistLimitDegrees=45
  twistMotion=Free
 }
 {
  bone1Name=neck_01
  bone2Name=spine_05
  swing1LimitDegrees=45
  swing1Motion=Free
  swing2LimitDegrees=45
  swing2Motion=Free
  twistLimitDegrees=45
  twistMotion=Free
 }
 {
  bone1Name=clavicle_l
  bone2Name=spine_05
  swing1LimitDegrees=45
  swing1Motion=Free
  swing2LimitDegrees=45
  swing2Motion=Free
  twistLimitDegrees=45
  twistMotion=Free
 }
 {
  bone1Name=spine_05
  bone2Name=spine_04
  swing1LimitDegrees=45
  swing1Motion=Free
  swing2LimitDegrees=45
  swing2Motion=Free
  twistLimitDegrees=45
  twistMotion=Free
 }
 {
  bone1Name=upperarm_r
  bone2Name=clavicle_r
  swing1LimitDegrees=45
  swing1Motion=Free
  swing2LimitDegrees=45
  swing2Motion=Free
  twistLimitDegrees=45
  twistMotion=Free
 }
]
