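# Post-process-style animation blueprint for the Sailor Squad Leader female body part, driving dynamic hair over poses copied from the parent and body meshes.
# Posed back-hair bones and front-hair simulation feed RigidBody variants chosen for front end, cape or backbling.
# Without rigid-body physics, Look At and movement-weighted Trail controllers take over.
# The event graph updates gravity, skydive scaling, targeting and head-rotation alphas, and resets dynamics after emotes.
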AnimBlueprint /BRCosmetics/Characters/Player/Female/Medium/Bodies/F_MED_Sailor_Squad_Leader/Meshes/Parts/F_MED_Sailor_Squad_Leader_AnimBP
target_skeleton=&/F_MED_Sailor_Squad_Leader_Skeleton
graphs [
 AnimGraph {nodes [
   n1 = Get Component Linear Vel Scale() -> (Component Linear Vel Scale:Vector)
   RigidBody!(
    ComponentPose
    Alpha
    OverrideWorldGravity:Vector=Get GravityOverride()
    ExternalForce
    ComponentLinearAccScale:Vector=@n1
    ComponentLinearVelScale:Vector=@n1
   ) -> (Pose) {
    bClampLinearTranslationLimitToRefPose=true
    bOverrideWorldGravity=true
    overridePhysicsAsset=&/F_MED_Sailor_Squad_Leader_dyn_hairFront
    simulationSpace=BaseBoneSpace
    alphaBoolBlend {blendOption=Linear}
    alphaScaleBiasClamp {bInterpResult=true, interpSpeedIncreasing=25}
    baseBoneRef {boneName=spine_05}
    componentAppliedLinearAccClamp=<3000 3000 3000>
    componentLinearAccScale=<0.32 0.2 0.2>
    componentLinearVelScale=<0.3 0.2 0.2>
    overrideWorldGravity {z=-780}
    simSpaceSettings {
     maxAngularAcceleration=1000
     maxAngularVelocity=3
     maxLinearAcceleration=700
     maxLinearVelocity=500
     worldAlpha=0.1
     externalLinearDragV=<0.2 0.2>
    }
   }
   n2 = "Transform (Modify) Bone - Bone: dyn_hair_mid_bk_MSTR"(
    ComponentPose=Local To Component(LocalPose=Use cached pose 'noRigidCache'())
    Alpha=1
    Translation="0,0,0"
    Rotation="0.000,5,18"
    Scale="1,1,1"
   ) {
    rotationMode=BMM_Additive
    rotationSpace=BCS_BoneSpace
    translationMode=BMM_Additive
    translationSpace=BCS_BoneSpace
    alphaBoolBlend {blendOption=Linear}
    boneToModify {boneName=dyn_hair_mid_bk_MSTR}
    rotation {roll=18, yaw=5}
   }
   n3 = "Transform (Modify) Bone - Bone: dyn_hair_r_bk_MSTR"(ComponentPose=@n2, Alpha=1, Translation="0,0,0", Rotation="-4.909,4.152,9.610", Scale="1,1,1") {
    rotationMode=BMM_Additive
    rotationSpace=BCS_BoneSpace
    translationMode=BMM_Additive
    translationSpace=BCS_BoneSpace
    alphaBoolBlend {blendOption=Linear}
    boneToModify {boneName=dyn_hair_r_bk_MSTR}
    rotation=<rot -4.909 4.152 9.61>
   }
   n4 = "Transform (Modify) Bone - Bone: dyn_hair_l_bk_MSTR"(ComponentPose=@n3, Alpha=1, Translation="0,0,0", Rotation="-5,-5,20", Scale="1,1,1") {
    rotationMode=BMM_Additive
    rotationSpace=BCS_BoneSpace
    translationMode=BMM_Additive
    translationSpace=BCS_BoneSpace
    alphaBoolBlend {blendOption=Linear}
    boneToModify {boneName=dyn_hair_l_bk_MSTR}
    rotation=<rot -5 -5 20>
   }
   Output Pose(Result=Blend Poses by bool(BlendPose_0=Use cached pose 'rigidCache'(), BlendPose_1=Component To Local(ComponentPose=@n4), bActiveValue=Get bRigidBodyEnabled()) {blendType=Linear}) {layerGroup=None}
   n5 = Get BwdMovement() -> (BwdMovement:double)
   n6 = Look At - Bone: dyn_hair_mid_01_bk(ComponentPose=Local To Component(LocalPose=Use cached pose 'StartPose'()), Alpha=1) {
    alphaBoolBlend {blendOption=Linear}
    boneToModify {boneName=dyn_hair_mid_01_bk}
    lookAtLocation=<-60 -30 1>
    lookAtTarget {boneReference {boneName=spine_05}}
    lookAt_Axis {axis=<1 0>}
    lookUp_Axis {axis {x=0, z=1}}
   }
   n7 = Look At - Bone: dyn_hair_mid_03_bk(ComponentPose=@n6, Alpha=1) {
    alphaBoolBlend {blendOption=Linear}
    boneToModify {boneName=dyn_hair_mid_03_bk}
    lookAtLocation=<-60 -70 1>
    lookAtTarget {boneReference {boneName=spine_05}}
    lookAt_Axis {axis=<1 0>}
    lookUp_Axis {axis {x=0, z=1}}
   }
   n8 = Look At - Bone: dyn_hair_l_01_bk(ComponentPose=@n7, Alpha=1) {
    alphaBoolBlend {blendOption=Linear}
    boneToModify {boneName=dyn_hair_l_01_bk}
    lookAtLocation=<-60 -20 -18>
    lookAtTarget {boneReference {boneName=spine_05}}
    lookAt_Axis {axis=<1 0>}
    lookUp_Axis {axis {x=0, z=1}}
   }
   n9 = Look At - Bone: dyn_hair_l_03_bk(ComponentPose=@n8, Alpha=1) {
    alphaBoolBlend {blendOption=Linear}
    boneToModify {boneName=dyn_hair_l_03_bk}
    lookAtLocation=<-60 -60 -8>
    lookAtTarget {boneReference {boneName=spine_05}}
    lookAt_Axis {axis=<1 0>}
    lookUp_Axis {axis {x=0, z=1}}
   }
   n10 = Look At - Bone: dyn_hair_r_01_bk(ComponentPose=@n9, Alpha=1) {
    alphaBoolBlend {blendOption=Linear}
    boneToModify {boneName=dyn_hair_r_01_bk}
    lookAtLocation=<-60 -30 10>
    lookAtTarget {boneReference {boneName=spine_05}}
    lookAt_Axis {axis=<1 0>}
    lookUp_Axis {axis {x=0, z=1}}
   }
   n11 = Look At - Bone: dyn_hair_r_03_bk(ComponentPose=@n10, Alpha=1) {
    alphaBoolBlend {blendOption=Linear}
    boneToModify {boneName=dyn_hair_r_03_bk}
    lookAtLocation=<-60 -70 40>
    lookAtTarget {boneReference {boneName=spine_05}}
    lookAt_Axis {axis=<1 0>}
    lookUp_Axis {axis {x=0, z=1}}
   }
   n12 = Look At - Bone: dyn_hair_01_l(ComponentPose=@n11, Alpha=1) {
    alphaBoolBlend {blendOption=Linear}
    boneToModify {boneName=dyn_hair_01_l}
    lookAtLocation=<-60 12 -10>
    lookAtTarget {boneReference {boneName=spine_05}}
    lookAt_Axis {axis=<1 0>}
    lookUp_Axis {axis {x=0, z=1}}
   }
   n13 = Look At - Bone: dyn_hair_01_r(ComponentPose=@n12, Alpha=1) {
    alphaBoolBlend {blendOption=Linear}
    boneToModify {boneName=dyn_hair_01_r}
    lookAtLocation=<-60 12 10>
    lookAtTarget {boneReference {boneName=spine_05}}
    lookAt_Axis {axis=<1 0>}
    lookUp_Axis {axis {x=0, z=1}}
   }
   n14 = Trail controller - Bone: dyn_hair_mid_05_bk(ComponentPose=@n13, Alpha=@n5) {
    bUsePlanarLimit=true
    chainLength=3
    debugLifeTime=0.1
    maxDeltaTime=0.0333
    planarLimits [{drivingBone {boneName=spine_05}, planeTransform {location=<8 -10 0>, rotation=<rot 0 25 -90>, scale=<0.4 0.5 0.5>}}]
    rotationLimits [{limitMax=<180 180 180>}, {limitMax=<180 180 180>}, {limitMax=<180 180 180>}]
    rotationOffsets [<0 0 0>, <0 0 0>, <0 0 0>]
    trailBone {boneName=dyn_hair_mid_05_bk}
    trailRelaxationSpeed {editorCurveData {keys [<at 0 15 RCIM_Cubic>, {arriveTangent=-10.7093, interpMode=RCIM_Cubic, leaveTangent=-10.7093, tangentMode=RCTM_User, time=1, value=6}]}}
   }
   n15 = Trail controller - Bone: dyn_hair_l_04_bk(ComponentPose=@n14, Alpha=@n5) {
    bUsePlanarLimit=true
    debugLifeTime=0.1
    maxDeltaTime=0.0333
    planarLimits [{drivingBone {boneName=spine_05}, planeTransform {location=<8 -10 0>, rotation=<rot 0 25 -89.9999>, scale=<0.4 0.5 0.5>}}]
    rotationLimits [{limitMax=<180 180 180>}, {limitMax=<180 180 180>}]
    rotationOffsets [<0 0 0>, <0 0 0>]
    trailBone {boneName=dyn_hair_l_04_bk}
    trailRelaxationSpeed {editorCurveData {keys [<at 0 15 RCIM_Cubic>, {arriveTangent=-10.7093, interpMode=RCIM_Cubic, leaveTangent=-10.7093, tangentMode=RCTM_User, time=1, value=6}]}}
   }
   n16 = Trail controller - Bone: dyn_hair_r_04_bk(ComponentPose=@n15, Alpha=@n5) {
    bUsePlanarLimit=true
    debugLifeTime=0.1
    maxDeltaTime=0.0333
    planarLimits [{drivingBone {boneName=spine_05}, planeTransform {location=<8 -10 0>, rotation=<rot 0 25 -89.9999>, scale=<0.4 0.5 0.5>}}]
    rotationLimits [{limitMax=<180 180 180>}, {limitMax=<180 180 180>}]
    rotationOffsets [<0 0 0>, <0 0 0>]
    trailBone {boneName=dyn_hair_r_04_bk}
    trailRelaxationSpeed {editorCurveData {keys [<at 0 15 RCIM_Cubic>, {arriveTangent=-10.7093, interpMode=RCIM_Cubic, leaveTangent=-10.7093, tangentMode=RCTM_User, time=1, value=6}]}}
   }
   n17 = Get FwdMovement() -> (FwdMovement:double)
   n18 = Trail controller - Bone: dyn_hair_04_l(ComponentPose=@n16, Alpha=@n17) {
    chainLength=4
    debugLifeTime=0.1
    maxDeltaTime=0.0333
    alphaScaleBiasClamp {bInterpResult=true}
    rotationLimits [{limitMax=<180 180 180>}, {limitMax=<180 180 180>}, {limitMax=<180 180 180>}, {limitMax=<180 180 180>}]
    rotationOffsets [<0 0 0>, <0 0 0>, <0 0 0>, <0 0 0>]
    trailBone {boneName=dyn_hair_04_l}
    trailRelaxationSpeed {editorCurveData {keys [<at 0 15 RCIM_Cubic>, {arriveTangent=-10.7093, interpMode=RCIM_Cubic, leaveTangent=-10.7093, tangentMode=RCTM_User, time=1, value=6}]}}
   }
   n19 = Trail controller - Bone: dyn_hair_04_r(ComponentPose=@n18, Alpha=@n17) {
    chainLength=4
    debugLifeTime=0.1
    maxDeltaTime=0.0333
    alphaScaleBiasClamp {bInterpResult=true}
    rotationLimits [{limitMax=<180 180 180>}, {limitMax=<180 180 180>}, {limitMax=<180 180 180>}, {limitMax=<180 180 180>}]
    rotationOffsets [<0 0 0>, <0 0 0>, <0 0 0>, <0 0 0>]
    trailBone {boneName=dyn_hair_04_r}
    trailRelaxationSpeed {editorCurveData {keys [<at 0 15 RCIM_Cubic>, {arriveTangent=-10.7093, interpMode=RCIM_Cubic, leaveTangent=-10.7093, tangentMode=RCTM_User, time=1, value=6}]}}
   }
   Save cached pose 'noRigidCache'(Pose=Component To Local(ComponentPose=@n19)) {cacheName=noRigidCache}
   n20 = RigidBody(
    ComponentPose=Local To Component(LocalPose=Use cached pose 'Hair Back'())
    OverrideWorldGravity="0,0,-980"
    ComponentLinearAccScale=@n1
    ComponentLinearVelScale=@n1
   ) {
    bClampLinearTranslationLimitToRefPose=true
    bOverrideWorldGravity=true
    overridePhysicsAsset=&/F_MED_Sailor_Squad_Leader_dyn_hair
    simulationSpace=BaseBoneSpace
    alphaBoolBlend {blendOption=Linear}
    alphaScaleBiasClamp {bInterpResult=true, interpSpeedIncreasing=25}
    baseBoneRef {boneName=spine_05}
    componentAppliedLinearAccClamp=<3000 3000 3000>
    componentLinearAccScale=<0.32 0.2 0.2>
    componentLinearVelScale=<0.3 0.2 0.2>
    overrideWorldGravity {z=-780}
    simSpaceSettings {maxAngularAcceleration=1000, maxAngularVelocity=3, maxLinearAcceleration=700, maxLinearVelocity=500}
   }
   n21 = Get Linear Scale WOP() -> (Linear Scale WOP:Vector)
   n22 = RigidBody(
    ComponentPose=Local To Component(LocalPose=Use cached pose 'Hair Back'())
    OverrideWorldGravity=Get GravityOverride()
    ComponentLinearAccScale=@n21
    ComponentLinearVelScale=@n21
   ) {
    bClampLinearTranslationLimitToRefPose=true
    bOverrideWorldGravity=true
    overridePhysicsAsset=&/F_MED_Sailor_Squad_Leader_dyn_hair_WOBackPack
    simulationSpace=BaseBoneSpace
    alphaBoolBlend {blendOption=Linear}
    alphaScaleBiasClamp {bInterpResult=true, interpSpeedIncreasing=25}
    baseBoneRef {boneName=spine_05}
    componentAppliedLinearAccClamp=<3000 3000 3000>
    componentLinearAccScale=<0.5 0.2 0.2>
    componentLinearVelScale=<0.5 0.2 0.2>
    overrideWorldGravity {z=-780}
   }
   n23 = Blend Poses by bool(
    BlendPose_0=Component To Local(ComponentPose=@n20)
    BlendPose_1=Component To Local(ComponentPose=@n22)
    bActiveValue=Get isWearingBackbling()
   ) {blendType=Linear}
   n24 = Get Linear Scale Cape() -> (Linear Scale Cape:Vector)
   n25 = RigidBody(
    ComponentPose=Local To Component(LocalPose=Use cached pose 'Hair Back'())
    OverrideWorldGravity=Get GravityOverride()
    ComponentLinearAccScale=@n24
    ComponentLinearVelScale=@n24
   ) {
    bClampLinearTranslationLimitToRefPose=true
    bOverrideWorldGravity=true
    overridePhysicsAsset=&/F_MED_Sailor_Squad_Leader_dyn_hair_Cape
    simulationSpace=BaseBoneSpace
    alphaBoolBlend {blendOption=Linear}
    alphaScaleBiasClamp {bInterpResult=true, interpSpeedIncreasing=25}
    baseBoneRef {boneName=neck_01}
    componentAppliedLinearAccClamp=<3000 3000 3000>
    componentLinearAccScale=<0.5 0.2 0.2>
    componentLinearVelScale=<0.5 0.2 0.2>
    overrideWorldGravity {z=-780}
   }
   n26 = Component To Local(ComponentPose=@n25)
   n27 = Get Is Wearing Cape() -> (Is Wearing Cape:bool)
   n28 = RigidBody!(ComponentPose=Local To Component(LocalPose=Use cached pose 'StartPose'()), Alpha, OverrideWorldGravity=Get GravityOverride(), ExternalForce) -> (Pose) {
    bClampLinearTranslationLimitToRefPose=true
    bOverrideWorldGravity=true
    overridePhysicsAsset=&/F_MED_Sailor_Squad_Leader_FrontEnd_hair
    alphaBoolBlend {blendOption=Linear}
    baseBoneRef {boneName=spine_05}
    componentAppliedLinearAccClamp=<6000 6000 6000>
    componentLinearAccScale=<0.1 0.1 0.1>
    componentLinearVelScale=<0.1 0.1 0.1>
   }
   Save cached pose 'rigidCache'(Pose=Blend Poses by bool(
     BlendPose_0=Component To Local(ComponentPose=@n28)
     BlendPose_1=Blend Poses by bool(BlendPose_0=@n26, BlendPose_1=@n23, bActiveValue=@n27) {blendType=Linear}
     bActiveValue=Get bIsInFrontEnd()
    ) {blendType=Linear}) {cacheName=rigidCache}
   n29 = "Transform (Modify) Bone - Bone: dyn_hair_r_bk_MSTR"(
    ComponentPose=Local To Component(LocalPose=Use cached pose 'StartPose'())
    Alpha=Get IsTargetingAlpha() -> (IsTargetingAlpha:double)
    Translation="-2,0,2"
    Rotation="30,0,0"
    Scale="1,1,1"
   ) {
    rotationMode=BMM_Additive
    rotationSpace=BCS_BoneSpace
    translationMode=BMM_Additive
    translationSpace=BCS_BoneSpace
    alphaBoolBlend {blendOption=Linear}
    alphaScaleBiasClamp {bInterpResult=true}
    boneToModify {boneName=dyn_hair_r_bk_MSTR}
    rotation {pitch=30}
    translation {x=-2, z=2}
   }
   n30 = Get IsNotSkydiveAlpha() -> (IsNotSkydiveAlpha:double)
   n31 = "Transform (Modify) Bone - Bone: dyn_hair_mid_bk_MSTR"(ComponentPose=@n29, Alpha=@n30, Translation="0,0,0", Rotation="45,0,0", Scale="1,1,1") {
    rotationMode=BMM_Additive
    rotationSpace=BCS_BoneSpace
    alphaBoolBlend {blendOption=Linear}
    boneToModify {boneName=dyn_hair_mid_bk_MSTR}
    rotation {pitch=45}
   }
   n32 = "Transform (Modify) Bone - Bone: dyn_hair_r_bk_MSTR"(ComponentPose=@n31, Alpha=@n30, Translation="0,0,0", Rotation="45,0,0", Scale="1,1,1") {
    rotationMode=BMM_Additive
    rotationSpace=BCS_BoneSpace
    alphaBoolBlend {blendOption=Linear}
    boneToModify {boneName=dyn_hair_r_bk_MSTR}
    rotation {pitch=45}
   }
   n33 = "Transform (Modify) Bone - Bone: dyn_hair_l_bk_MSTR"(ComponentPose=@n32, Alpha=@n30, Translation="0,0,0", Rotation="45,0,0", Scale="1,1,1") {
    rotationMode=BMM_Additive
    rotationSpace=BCS_BoneSpace
    alphaBoolBlend {blendOption=Linear}
    alphaScaleBiasClamp {bInterpResult=true}
    boneToModify {boneName=dyn_hair_l_bk_MSTR}
    rotation {pitch=45}
   }
   n34 = Get bIsSkydiveFloating()
   n35 = "Transform (Modify) Bone - Bone: dyn_hair_mid_bk_MSTR"!(ComponentPose=@n33, bAlphaBoolEnabled:bool=@n34, Translation="0,0,0", Rotation="10,0,0", Scale="1,1,1") -> (Pose) {
    alphaInputType=Bool
    rotationMode=BMM_Additive
    rotationSpace=BCS_BoneSpace
    alphaBoolBlend {blendOption=Linear}
    alphaScaleBiasClamp {bInterpResult=true, interpSpeedIncreasing=5}
    boneToModify {boneName=dyn_hair_mid_bk_MSTR}
    rotation {pitch=10}
   }
   n36 = "Transform (Modify) Bone - Bone: dyn_hair_r_bk_MSTR"!(ComponentPose=@n35, bAlphaBoolEnabled:bool=@n34, Translation="0,0,0", Rotation="10,0,0", Scale="1,1,1") -> (Pose) {
    alphaInputType=Bool
    rotationMode=BMM_Additive
    rotationSpace=BCS_BoneSpace
    alphaBoolBlend {blendOption=Linear}
    alphaScaleBiasClamp {bInterpResult=true, interpSpeedIncreasing=5}
    boneToModify {boneName=dyn_hair_r_bk_MSTR}
    rotation {pitch=10}
   }
   n37 = "Transform (Modify) Bone - Bone: dyn_hair_l_bk_MSTR"!(ComponentPose=@n36, bAlphaBoolEnabled:bool=@n34, Translation="0,0,0", Rotation="10,0,0", Scale="1,1,1") -> (Pose) {
    alphaInputType=Bool
    rotationMode=BMM_Additive
    rotationSpace=BCS_BoneSpace
    alphaBoolBlend {blendOption=Linear}
    alphaScaleBiasClamp {bInterpResult=true, interpSpeedIncreasing=5}
    boneToModify {boneName=dyn_hair_l_bk_MSTR}
    rotation {pitch=10}
   }
   n38 = Look At - Bone: dyn_hair_r_bk_MSTR(ComponentPose=@n37, Alpha=1) {
    alphaBoolBlend {blendOption=Linear}
    boneToModify {boneName=dyn_hair_r_bk_MSTR}
    lookAtLocation=<-60 -30 10>
    lookAtTarget {boneReference {boneName=spine_05}}
    lookAt_Axis {axis=<1 0>}
    lookUp_Axis {axis {x=0, z=1}}
   }
   n39 = Look At - Bone: dyn_hair_mid_bk_MSTR(ComponentPose=@n38, Alpha=1) {
    alphaBoolBlend {blendOption=Linear}
    boneToModify {boneName=dyn_hair_mid_bk_MSTR}
    lookAtLocation=<-60 -30 1>
    lookAtTarget {boneReference {boneName=spine_05}}
    lookAt_Axis {axis=<1 0>}
    lookUp_Axis {axis {x=0, z=1}}
   }
   n40 = Look At - Bone: dyn_hair_l_bk_MSTR(ComponentPose=@n39, Alpha=1) {
    alphaBoolBlend {blendOption=Linear}
    boneToModify {boneName=dyn_hair_l_bk_MSTR}
    lookAtLocation=<-60 -20 -18>
    lookAtTarget {boneReference {boneName=spine_05}}
    lookAt_Axis {axis=<1 0>}
    lookUp_Axis {axis {x=0, z=1}}
   }
   n41 = Get Component Linear Vel Scale()
   n42 = RigidBody!(
    ComponentPose=@n40
    Alpha
    OverrideWorldGravity=Get GravityOverride()
    ExternalForce
    ComponentLinearAccScale:Vector=@n41
    ComponentLinearVelScale:Vector=@n41
   ) -> (Pose) {
    bClampLinearTranslationLimitToRefPose=true
    bOverrideWorldGravity=true
    overridePhysicsAsset=&F_MED_Sailor_Squad_Leader_dyn_hairFront
    simulationSpace=BaseBoneSpace
    alphaBoolBlend {blendOption=Linear}
    alphaScaleBiasClamp {bInterpResult=true, interpSpeedIncreasing=25}
    baseBoneRef {boneName=spine_05}
    componentAppliedLinearAccClamp=<3000 3000 3000>
    componentLinearAccScale=<0.32 0.2 0.2>
    componentLinearVelScale=<0.3 0.2 0.2>
    overrideWorldGravity {z=-780}
    simSpaceSettings {
     maxAngularAcceleration=1000
     maxAngularVelocity=3
     maxLinearAcceleration=700
     maxLinearVelocity=500
     worldAlpha=0.1
     externalLinearDragV=<0.2 0.2>
    }
   }
   Save cached pose 'Hair Back'(Pose=Component To Local(ComponentPose=@n42)) {cacheName=Hair Back}
   n43 = Layered blend per bone(
    BasePose=Copy Pose From Mesh() {bCopyCurves=true, bUseAttachedParent=true}
    BlendPoses_0=Copy Pose From Mesh(SourceMeshComponent=Get Body_Mesh())
   ) {layerSetup [{branchFilters [{boneName=clavicle_l}, {boneName=clavicle_r}, {boneName=upperarm_r}, {boneName=upperarm_l}]}]}
   Save cached pose 'StartPose'(Pose=@n43) {cacheName=StartPose}
  ]}
 EventGraph {nodes [
   Event OnCharacterCustomizationLoaded()
   > n1 = Cast To PlayerPawn_Athena(Object=TryGetPawnOwner())
   > Set Body_Mesh(Body_Mesh=GetSkeletalMeshForPartType(self=@n1))
   Event BlueprintUpdateAnimation()
   > n2 = Sequence!(execute) -> (then_0, then_1, then_2:Exec, then_3:Exec, then_4:Exec, then_5:Exec, then_6:Exec)
   n3 = PlayerMovement(execute=@n2.then_1, AnimInstance=Self-Reference(), XDotProductCoefficient=1, YDotProductCoefficient=1, ZDotProductCoefficient=1)
   n4 = MapRangeClamped(Value=@n3.YMovement, InRangeA=0, InRangeB=1)
   > Set FwdMovement(FwdMovement:double=MapRangeClamped(Value=float + float(A=@n3, B=@n4) {is=title K2Node_CommutativeAssociativeBinaryOperator}, InRangeA=0, InRangeB=1, OutRangeA=0.3, OutRangeB=0)) -> (Output_Get:double)
   n5 = "Absolute (Float)"(A=@n3)
   n6 = MapRangeClamped(Value=@n3.YMovement, InRangeA=0, InRangeB=-1)
   > Set BwdMovement(BwdMovement:double=MapRangeClamped(Value=float + float(A=@n5, B=@n6), InRangeA=0, InRangeB=1, OutRangeA=0.3, OutRangeB=0)) -> (Output_Get:double)
   n7 = Branch(execute=@n2.then_3, Condition=Get IsSkydive() -> (IsSkydive:bool))
   > Set Component Linear Acc Scale(Component Linear Acc Scale:Vector="0.1,0.1,0.1") -> (Output_Get:Vector)
   > Set Component Linear Vel Scale(Component Linear Vel Scale:Vector="0.1,0.1,0.1") -> (Output_Get:Vector)
   > Set Linear Scale WOP(Linear Scale WOP:Vector="0.1,0.2,0.2") -> (Output_Get:Vector)
   > Set Linear Scale Cape(Linear Scale Cape:Vector="0.4,0.2,0.2") -> (Output_Get:Vector)
   n8 = Set Component Linear Acc Scale(execute=@n7.else, Component Linear Acc Scale="0,0.1,0.1")
   > Set Component Linear Vel Scale(Component Linear Vel Scale=@n8)
   > Set Linear Scale WOP(Linear Scale WOP="0.4,0.2,0.2")
   > Set Linear Scale Cape(Linear Scale Cape="0.4,0.2,0.2")
   n9 = Get bIsSkydiveFloating()
   n10 = Get bIsSkydiveDiveMode()
   n11 = XOR Boolean!(A:bool=NOT Boolean(A=@n9), B:bool=@n10) -> (ReturnValue:bool)
   n12 = Get head_rot_bk() -> (head_rot_bk:double)
   Set IsNotSkydiveAlpha(
    execute=@n2.then_4
    IsNotSkydiveAlpha:double=float * float(A=Remap Float(Input=@n12, OldRange=(X=0.2,Y=1), NewRange=(X=0.7,Y=1)) -> (Output=~), B="To Float (Boolean)"(InBool=@n11)) {is=title K2Node_PromotableOperator}
   ) -> (Output_Get:double)
   n13 = OR Boolean(A=@n9, B=@n10)
   Set SkydiveFloaringBin(execute=@n2.then_5, SkydiveFloaringBin:double=SelectFloat(A=0.025, B=1, bPickA=@n13)) -> (Output_Get:double)
   > Set IsSkydive(IsSkydive:bool=OR Boolean(A=@n13, B=Get bIsParachuteOpened())) -> (Output_Get:bool)
   n14 = Branch(execute=@n2.then_6, Condition=Get bIsPlayingEmote())
   Do Once(execute=@n14.else, Reset=@n14.then, Start Closed=true)
   > Delay(Duration=0.2)
   > ResetDynamics(InTeleportType=TeleportPhysics)
   > ResetDynamics(InTeleportType=ResetPhysics)
   n15 = Gravity Override State Function DEPRECATED(
    execute=@n2.then_0
    CharacterPartAnimInstance=Self-Reference()
    GameTimeInSeconds=GetGameTimeInSeconds()
    GravityOverrideParams=Get GravityOverrideParams() -> (GravityOverrideParams:Gravity Override Params Struct Structure)
   )
   > Set GravityOverride(GravityOverride=@n15)
   n16 = CheckForBackbling()
   > Set isWearingBackbling(isWearingBackbling=AND Boolean!(A=@n16, B=NOT Boolean(A=@n16.bIsBackblingACape), C:bool=NOT Boolean(A=@n16.bIsBackblingLong)) -> (ReturnValue))
   > Set Is Wearing Cape(Is Wearing Cape=OR Boolean(A=@n16.bIsBackblingACape, B=@n16.bIsBackblingLong))
   Set head_rot_bk(
    execute=@n2.then_2
    head_rot_bk:double=GetAlphaFromReferenceTransformDelta(SkeletalMeshComponent=Get Body_Mesh(), BoneName=spine_02, BaseName=head, InputRange=(X=-5,Y=10), OutputRange=(X=0.2,Y=1), TwistAxis="0,0,1")
   ) -> (Output_Get:double)
   n17 = Get bIsTargeting()
   > Set IsTargetingAlpha(IsTargetingAlpha:double=float * float(A=Get head_rot_bk(), B="To Float (Boolean)"(InBool=@n17))) -> (Output_Get:double)
  ]}
]
variables [
 Body_Mesh {type=SkeletalMeshComponent}
 BwdMovement {type=number}
 Component Linear Acc Scale {type=Vector, default=<0.32 0.2 0.2>}
 Component Linear Vel Scale {type=Vector, default=<0.3 0.2 0.2>}
 FwdMovement {type=number}
 gravity_override {type=Vector, default=<0 0 -780>}
 GravityOverride {type=Vector}
 GravityOverrideParams {
  type=GravityOverrideParamsStruct
  default {
   emoteOrMelee {gravityOverride=<0 0 -1500>}
   front_End {gravityOverride=<0 0 -1500>}
   skydive_FloatingDown {gravityOverride=<0 -700 500>}
   skydive_GoingUp {gravityOverride=<0 -300 -500>}
  }
 }
 head_rot_bk {type=number}
 Is Wearing Cape {type=boolean}
 IsNotSkydiveAlpha {type=number}
 IsSkydive {type=boolean}
 IsTargetingAlpha {type=number}
 isWearingBackbling {type=boolean}
 JumpingAlpha {type=number}
 Linear Scale Cape {type=Vector}
 Linear Scale WOP {type=Vector}
 SkydiveFloaringBin {type=number}
]
